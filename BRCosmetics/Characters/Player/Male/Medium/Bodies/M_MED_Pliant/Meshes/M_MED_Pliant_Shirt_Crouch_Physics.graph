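# Physics asset for secondary motion of the Pliant medium male body's shirt, in a crouch-specific variant.
# Four swing- and twist-limited chains of boxes or capsules hang from tiny MSTR anchors linked to the left thigh, right thigh and pelvis.
# Pelvis, thigh, calf and right-foot capsules provide body shapes, joined by free-rotation constraints.
# A large, thin, unconstrained box sits on dyn_root_simspace, apparently as a ground plane.
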PhysicsAsset /BRCosmetics/Characters/Player/Male/Medium/Bodies/M_MED_Pliant/Meshes/M_MED_Pliant_Shirt_Crouch_Physics
bodies {
 calf_l [{shapeType=Capsule, length=39.905, radius=8.1775, center=<-25.7712 0.002 0>, rotation=<rot 90 0 0>}]
 calf_r [{shapeType=Capsule, length=28.2685, radius=11.0286, center=<23.7907 0.0021 -0.0002>, rotation=<rot 90 0 0>}]
 dyn_root_simspace [{
   shapeType=Box
   extentX=305.5876
   extentY=264.2872
   extentZ=24.352
   shapeName=dyn_root_simspace_box
   center=<0 -0.0004 -4.72>
   rotation=<rot 0 -90 0>
  }]
 dyn_shirt_a_01 [{shapeType=Box, extentX=5.5198, extentY=13.8328, extentZ=2, center=<2.9797 0 -0.0003>}]
 dyn_shirt_a_02 [{shapeType=Box, extentX=6.3914, extentY=13.8328, extentZ=2, center=<3.1248 0 -0.0004>}]
 dyn_shirt_a_MSTR [{shapeType=Capsule, length=0.505, radius=0.505}]
 dyn_shirt_b_01 [{shapeType=Box, extentX=9.7529, extentY=13.8328, extentZ=2.9603, center=<5.1042 0 0>}]
 dyn_shirt_b_02 [{shapeType=Box, extentX=8.8166, extentY=13.8328, extentZ=2.9603, center=<4.4769 0 -0.0004>}]
 dyn_shirt_b_MSTR [{shapeType=Capsule, length=0.505, radius=0.505}]
 dyn_shirt_c_01 [{shapeType=Capsule, length=20, radius=5, shapeName=dyn_shirt_c_01_capsule, center=<6.2749 0 0>, rotation=<rot 81.6294 89.9999 -0.0001>}]
 dyn_shirt_c_02 [{shapeType=Capsule, length=20, radius=5, shapeName=dyn_shirt_c_02_capsule, center=<6.2749 0 0>, rotation=<rot 81.6294 89.9999 -0.0001>}]
 dyn_shirt_c_03 [{shapeType=Capsule, length=20, radius=5, shapeName=dyn_shirt_c_03_capsule, center=<6.2749 0 0>, rotation=<rot 81.6294 89.9999 -0.0001>}]
 dyn_shirt_c_04 [{shapeType=Capsule, length=20, radius=5, shapeName=dyn_shirt_c_04_capsule, center=<6.2749 0 0>, rotation=<rot 81.6294 89.9999 -0.0001>}]
 dyn_shirt_c_MSTR [{shapeType=Capsule, length=0.505, radius=0.505}]
 dyn_shirt_d_01 [{shapeType=Box, extentX=9.6297, extentY=34.7561, extentZ=1.9301, center=<5.3106 0 0>}]
 dyn_shirt_d_02 [{shapeType=Box, extentX=12.1187, extentY=34.7561, extentZ=4.0332, center=<6.2924 0 -0.0001>}]
 dyn_shirt_d_03 [{shapeType=Box, extentX=8.4718, extentY=34.7561, extentZ=4.3905, center=<4.4383 0 0.0001>}]
 dyn_shirt_d_04 [{shapeType=Box, extentX=8.4718, extentY=34.7561, extentZ=5.0763, center=<3.9892 0 0.0001>}]
 dyn_shirt_d_MSTR [{shapeType=Capsule, length=0.505, radius=0.505}]
 foot_r [{shapeType=Capsule, length=13.5979, radius=6.7855, shapeName=foot_r_capsule, center=<5.4837 -5.5008 0>, rotation=<rot 0 0 90>}]
 pelvis [{shapeType=Capsule, length=2.9363, radius=13.2783, center=<-9.0622 0 0>}]
 thigh_l [{shapeType=Capsule, length=32.0316, radius=7.8834, center=<-21.0566 -1.3638 0.0001>, rotation=<rot 90 8.8807 2.9602>}]
 thigh_r [{shapeType=Capsule, length=48.0989, radius=8.5171, center=<33.0567 2.4403 -0.0254>, rotation=<rot 89.1338 179.9998 176.3289>}]
}
constraints [
 {bone1Name=dyn_shirt_a_01, bone2Name=dyn_shirt_a_MSTR, swing1LimitDegrees=45, swing2LimitDegrees=45, twistLimitDegrees=45}
 {bone1Name=dyn_shirt_a_02, bone2Name=dyn_shirt_a_01, swing1LimitDegrees=45, swing2LimitDegrees=45, twistLimitDegrees=45}
 {bone1Name=dyn_shirt_b_01, bone2Name=dyn_shirt_b_MSTR, swing1LimitDegrees=45, swing2LimitDegrees=45, twistLimitDegrees=45}
 {bone1Name=dyn_shirt_b_02, bone2Name=dyn_shirt_b_01, swing1LimitDegrees=45, swing2LimitDegrees=45, twistLimitDegrees=45}
 {bone1Name=dyn_shirt_c_01, bone2Name=dyn_shirt_c_MSTR, swing1LimitDegrees=10, swing2LimitDegrees=10, twistLimitDegrees=5}
 {bone1Name=dyn_shirt_c_02, bone2Name=dyn_shirt_c_01, swing1LimitDegrees=45, swing2LimitDegrees=45, twistLimitDegrees=5}
 {bone1Name=dyn_shirt_c_03, bone2Name=dyn_shirt_c_02, swing1LimitDegrees=45, swing2LimitDegrees=45, twistLimitDegrees=5}
 {bone1Name=dyn_shirt_c_04, bone2Name=dyn_shirt_c_03, swing1LimitDegrees=45, swing2LimitDegrees=45, twistLimitDegrees=5}
 {bone1Name=dyn_shirt_d_01, bone2Name=dyn_shirt_d_MSTR, swing1LimitDegrees=45, swing2LimitDegrees=45, twistLimitDegrees=5}
 {bone1Name=dyn_shirt_d_02, bone2Name=dyn_shirt_d_01, swing1LimitDegrees=45, swing2LimitDegrees=45, twistLimitDegrees=5}
 {bone1Name=dyn_shirt_d_03, bone2Name=dyn_shirt_d_02, swing1LimitDegrees=45, swing2LimitDegrees=45, twistLimitDegrees=5}
 {bone1Name=dyn_shirt_d_04, bone2Name=dyn_shirt_d_03, swing1LimitDegrees=45, swing2LimitDegrees=45, twistLimitDegrees=5}
 {
  bone1Name=dyn_shirt_a_MSTR
  bone2Name=thigh_l
  swing1LimitDegrees=45
  swing1Motion=Free
  swing2LimitDegrees=45
  swing2Motion=Free
  twistLimitDegrees=45
  twistMotion=Free
 }
 {
  bone1Name=dyn_shirt_b_MSTR
  bone2Name=thigh_r
  swing1LimitDegrees=45
  swing1Motion=Free
  swing2LimitDegrees=45
  swing2Motion=Free
  twistLimitDegrees=45
  twistMotion=Free
 }
 {
  bone1Name=dyn_shirt_c_MSTR
  bone2Name=thigh_r
  swing1LimitDegrees=45
  swing1Motion=Free
  swing2LimitDegrees=45
  swing2Motion=Free
  twistLimitDegrees=45
  twistMotion=Free
 }
 {
  bone1Name=thigh_l
  bone2Name=pelvis
  swing1LimitDegrees=45
  swing1Motion=Free
  swing2LimitDegrees=45
  swing2Motion=Free
  twistLimitDegrees=45
  twistMotion=Free
 }
 {
  bone1Name=thigh_r
  bone2Name=pelvis
  swing1LimitDegrees=45
  swing1Motion=Free
  swing2LimitDegrees=45
  swing2Motion=Free
  twistLimitDegrees=45
  twistMotion=Free
 }
 {
  bone1Name=dyn_shirt_d_MSTR
  bone2Name=pelvis
  swing1LimitDegrees=45
  swing1Motion=Free
  swing2LimitDegrees=45
  swing2Motion=Free
  twistLimitDegrees=45
  twistMotion=Free
 }
 {
  bone1Name=calf_r
  bone2Name=thigh_r
  swing1LimitDegrees=45
  swing1Motion=Free
  swing2LimitDegrees=45
  swing2Motion=Free
  twistLimitDegrees=45
  twistMotion=Free
 }
 {
  bone1Name=calf_l
  bone2Name=thigh_l
  swing1LimitDegrees=45
  swing1Motion=Free
  swing2LimitDegrees=45
  swing2Motion=Free
  twistLimitDegrees=45
  twistMotion=Free
 }
 {
  bone1Name=foot_r
  bone2Name=calf_r
  swing1LimitDegrees=45
  swing1Motion=Free
  swing2LimitDegrees=45
  swing2Motion=Free
  twistLimitDegrees=45
  twistMotion=Free
 }
]
