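# Physics asset for the shirt of the male medium Pliant player body, giving its dynamic shirt bones collision and joint limits.
# Four box-body chains (two short, two long) hang from small master capsules, with limited swing and twist, tighter twist on the long chains.
# The masters link with free angular motion to the left thigh, right thigh or pelvis.
# Capsules on pelvis, thighs and calves provide lower-body shapes, with the right thigh doubled.
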PhysicsAsset /BRCosmetics/Characters/Player/Male/Medium/Bodies/M_MED_Pliant/Meshes/M_MED_Pliant_Shirt_Physics
bodies {
 calf_l [{shapeType=Capsule, length=39.905, radius=8.1775, center=<-25.7712 0.002 0>, rotation=<rot 90 0 0>}]
 calf_r [{shapeType=Capsule, length=39.9755, radius=10.4284, center=<22.5424 0.1886 0.0003>, rotation=<rot 90 -90 -86.9017>}]
 dyn_shirt_a_01 [{shapeType=Box, extentX=5.5198, extentY=13.8328, extentZ=2, center=<2.9797 0 -0.0003>}]
 dyn_shirt_a_02 [{shapeType=Box, extentX=6.3914, extentY=13.8328, extentZ=2, center=<3.1248 0 -0.0004>}]
 dyn_shirt_a_MSTR [{shapeType=Capsule, length=0.505, radius=0.505}]
 dyn_shirt_b_01 [{shapeType=Box, extentX=9.7529, extentY=13.8328, extentZ=2.9603, center=<5.1042 0 0>}]
 dyn_shirt_b_02 [{shapeType=Box, extentX=8.8166, extentY=13.8328, extentZ=2.9603, center=<4.4769 0 -0.0004>}]
 dyn_shirt_b_MSTR [{shapeType=Capsule, length=0.505, radius=0.505}]
 dyn_shirt_c_01 [{shapeType=Box, extentX=9.8484, extentY=16.4611, extentZ=2.0778, center=<5.4757 -0.0001 0.847>, rotation=<rot -2.9439 0 -0.0001>}]
 dyn_shirt_c_02 [{shapeType=Box, extentX=7.8081, extentY=13.8438, extentZ=3.0246, center=<3.9489 0 -0.0004>}]
 dyn_shirt_c_03 [{shapeType=Box, extentX=9.0003, extentY=13.8438, extentZ=8.0785, center=<4.4268 0 -0.0004>}]
 dyn_shirt_c_04 [{shapeType=Box, extentX=8.4718, extentY=13.8328, extentZ=8.0675, center=<4.2271 0 -0.0004>}]
 dyn_shirt_c_MSTR [{shapeType=Capsule, length=0.505, radius=0.505}]
 dyn_shirt_d_01 [{shapeType=Box, extentX=9.6297, extentY=17.1658, extentZ=5.0079, center=<4.9985 0.0145 0.2674>, rotation=<rot -3.3158 -0.1806 3.1202>}]
 dyn_shirt_d_02 [{shapeType=Box, extentX=12.1187, extentY=17.1658, extentZ=7.2607, center=<6.056 0 -0.0003>}]
 dyn_shirt_d_03 [{shapeType=Box, extentX=8.4718, extentY=17.1658, extentZ=7.2607, center=<3.9489 0 -0.0004>}]
 dyn_shirt_d_04 [{shapeType=Box, extentX=8.4718, extentY=17.1658, extentZ=7.2607, center=<3.9489 0 -0.0004>}]
 dyn_shirt_d_MSTR [{shapeType=Capsule, length=0.505, radius=0.505}]
 pelvis [{shapeType=Capsule, length=2.9363, radius=13.2783, center=<-9.0622 0 0>}]
 thigh_l [{shapeType=Capsule, length=29.1006, radius=10.8145, center=<-21.0566 -1.3638 0.0001>, rotation=<rot 90 8.8807 2.9602>}]
 thigh_r [
  {shapeType=Capsule, length=32.695, radius=10.008, center=<25.022 0 0>, rotation=<rot 90 0 0>}
  {shapeType=Capsule, length=33.3783, radius=9.3246, shapeName=thigh_r_capsule, center=<25.1487 1.3671 -3.3067>, rotation=<rot 90 0 0>}
 ]
}
constraints [
 {bone1Name=dyn_shirt_a_01, bone2Name=dyn_shirt_a_MSTR, swing1LimitDegrees=45, swing2LimitDegrees=45, twistLimitDegrees=45}
 {bone1Name=dyn_shirt_a_02, bone2Name=dyn_shirt_a_01, swing1LimitDegrees=45, swing2LimitDegrees=45, twistLimitDegrees=45}
 {bone1Name=dyn_shirt_b_01, bone2Name=dyn_shirt_b_MSTR, swing1LimitDegrees=45, swing2LimitDegrees=45, twistLimitDegrees=45}
 {bone1Name=dyn_shirt_b_02, bone2Name=dyn_shirt_b_01, swing1LimitDegrees=45, swing2LimitDegrees=45, twistLimitDegrees=45}
 {bone1Name=dyn_shirt_c_01, bone2Name=dyn_shirt_c_MSTR, swing1LimitDegrees=45, swing2LimitDegrees=45, twistLimitDegrees=5}
 {bone1Name=dyn_shirt_c_02, bone2Name=dyn_shirt_c_01, swing1LimitDegrees=45, swing2LimitDegrees=45, twistLimitDegrees=5}
 {bone1Name=dyn_shirt_c_03, bone2Name=dyn_shirt_c_02, swing1LimitDegrees=45, swing2LimitDegrees=45, twistLimitDegrees=5}
 {bone1Name=dyn_shirt_c_04, bone2Name=dyn_shirt_c_03, swing1LimitDegrees=45, swing2LimitDegrees=45, twistLimitDegrees=5}
 {bone1Name=dyn_shirt_d_01, bone2Name=dyn_shirt_d_MSTR, swing1LimitDegrees=45, swing2LimitDegrees=45, twistLimitDegrees=5}
 {bone1Name=dyn_shirt_d_02, bone2Name=dyn_shirt_d_01, swing1LimitDegrees=45, swing2LimitDegrees=45, twistLimitDegrees=5}
 {bone1Name=dyn_shirt_d_03, bone2Name=dyn_shirt_d_02, swing1LimitDegrees=45, swing2LimitDegrees=45, twistLimitDegrees=5}
 {bone1Name=dyn_shirt_d_04, bone2Name=dyn_shirt_d_03, swing1LimitDegrees=45, swing2LimitDegrees=45, twistLimitDegrees=5}
 {
  bone1Name=dyn_shirt_a_MSTR
  bone2Name=thigh_l
  swing1LimitDegrees=45
  swing1Motion=Free
  swing2LimitDegrees=45
  swing2Motion=Free
  twistLimitDegrees=45
  twistMotion=Free
 }
 {
  bone1Name=dyn_shirt_b_MSTR
  bone2Name=thigh_r
  swing1LimitDegrees=45
  swing1Motion=Free
  swing2LimitDegrees=45
  swing2Motion=Free
  twistLimitDegrees=45
  twistMotion=Free
 }
 {
  bone1Name=dyn_shirt_c_MSTR
  bone2Name=thigh_r
  swing1LimitDegrees=45
  swing1Motion=Free
  swing2LimitDegrees=45
  swing2Motion=Free
  twistLimitDegrees=45
  twistMotion=Free
 }
 {
  bone1Name=thigh_l
  bone2Name=pelvis
  swing1LimitDegrees=45
  swing1Motion=Free
  swing2LimitDegrees=45
  swing2Motion=Free
  twistLimitDegrees=45
  twistMotion=Free
 }
 {
  bone1Name=thigh_r
  bone2Name=pelvis
  swing1LimitDegrees=45
  swing1Motion=Free
  swing2LimitDegrees=45
  swing2Motion=Free
  twistLimitDegrees=45
  twistMotion=Free
 }
 {
  bone1Name=dyn_shirt_d_MSTR
  bone2Name=pelvis
  swing1LimitDegrees=45
  swing1Motion=Free
  swing2LimitDegrees=45
  swing2Motion=Free
  twistLimitDegrees=45
  twistMotion=Free
 }
 {
  bone1Name=calf_r
  bone2Name=thigh_r
  swing1LimitDegrees=45
  swing1Motion=Free
  swing2LimitDegrees=45
  swing2Motion=Free
  twistLimitDegrees=45
  twistMotion=Free
 }
 {
  bone1Name=calf_l
  bone2Name=thigh_l
  swing1LimitDegrees=45
  swing1Motion=Free
  swing2LimitDegrees=45
  swing2Motion=Free
  twistLimitDegrees=45
  twistMotion=Free
 }
]
