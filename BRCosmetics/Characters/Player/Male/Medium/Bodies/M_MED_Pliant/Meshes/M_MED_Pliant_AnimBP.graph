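# Post-animation AnimBlueprint for the M_MED_Pliant male player body, adding secondary motion to jacket, shirt, collar and leg straps.
# It copies the parent mesh pose, blends in shirt and collar corrections when crouching, then runs three Control Rigs.
# Slope and sim-space root bones feed chained constraints, bone offsets, AnimDynamics and RigidBody simulations.
# Crouch, look-up and front-end variables adjust blend weights; jacket dynamics simulate relative to the sim-space bone.
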AnimBlueprint /BRCosmetics/Characters/Player/Male/Medium/Bodies/M_MED_Pliant/Meshes/M_MED_Pliant_AnimBP
target_skeleton=/Game/Characters/Player/Male/Medium/Bodies/M_MED_Pliant/Meshes/M_MED_Pliant_Skeleton
graphs [
 AnimGraph {nodes [
   n1 = Anim Dynamics - Bone: dyn_jacket_g_02(Alpha=1, ExternalForce="0,0,0") {
    angularDampingOverride=0.9999
    angularSpringConstant=25
    bAngularSpring=true
    bGravityOverrideInSimSpace=true
    bOverrideAngularDamping=true
    simulationSpace=BoneRelative
    alphaBoolBlend {blendOption=Linear}
    boundBone {boneName=dyn_jacket_g_02}
    componentAppliedLinearAccClamp=<3500 3500 3500>
    componentLinearAccScale=<1.5 1.5 1.5>
    componentLinearVelScale=<1.5 1.5 1.5>
    physicsBodyDefinitions [{
      sphereCollisionRadius=0
      boundBone {boneName=dyn_jacket_g_02}
      boxExtents=<10 20 20>
      constraintSetup {angularLimitsMax=<25 0 25>, angularLimitsMin=<-25 -25 -25>, angularTarget=<0 0 0>}
      localJointOffset=<2 0 0>
     }]
    relativeSpaceBone {boneName=dyn_root_simspace}
   }
   Anim Dynamics - Bone: dyn_jacket_g_03(ComponentPose=@n1, Alpha=1, ExternalForce="0,0,0") {
    angularDampingOverride=0.9999
    angularSpringConstant=25
    bAngularSpring=true
    bGravityOverrideInSimSpace=true
    bOverrideAngularDamping=true
    simulationSpace=BoneRelative
    alphaBoolBlend {blendOption=Linear}
    boundBone {boneName=dyn_jacket_g_03}
    componentAppliedLinearAccClamp=<3500 3500 3500>
    componentLinearAccScale=<1.5 1.5 1.5>
    componentLinearVelScale=<1.5 1.5 1.5>
    physicsBodyDefinitions [{
      sphereCollisionRadius=0
      boundBone {boneName=dyn_jacket_g_03}
      boxExtents=<10 20 20>
      constraintSetup {angularLimitsMax=<25 0 25>, angularLimitsMin=<-25 -25 -25>, angularTarget=<0 0 0>}
      localJointOffset=<2 0 0>
     }]
    relativeSpaceBone {boneName=dyn_root_simspace}
   }
   "Constraint - dyn_r_collar_MSTR (P)"(Alpha=1, ConstraintWeights_0=1) {
    alphaBoolBlend {blendOption=Linear}
    boneToModify {boneName=dyn_r_collar_MSTR}
    constraintSetup [{targetBone {boneName=neck_01}}]
    constraintWeights [1]
   }
   "Constraint - dyn_r_01_collar_MSTR (P)"(Alpha=1, ConstraintWeights_0=1) {
    alphaBoolBlend {blendOption=Linear}
    boneToModify {boneName=dyn_r_01_collar_MSTR}
    constraintSetup [{targetBone {boneName=neck_01}}]
    constraintWeights [1]
   }
   n2 = RigidBody!(
    ComponentPose=Local To Component(LocalPose=Copy Pose From Mesh() {bCopyCurves=true, bUseAttachedParent=true})
    bAlphaBoolEnabled:bool
    OverrideWorldGravity:Vector=Get GravityOverride()
    ExternalForce
   ) -> (Pose) {
    alphaInputType=Bool
    bClampLinearTranslationLimitToRefPose=true
    bOverrideWorldGravity=true
    overridePhysicsAsset=&/M_MED_Pliant_Shirt_Crouch_Physics
    simulationSpace=BaseBoneSpace
    alphaBoolBlend {blendInTime=0.1, blendOption=Linear, blendOutTime=0.1}
    baseBoneRef {boneName=root}
    componentAppliedLinearAccClamp=<5000 5000 5000>
    componentLinearAccScale=<0.5 0.5 0.5>
    overrideWorldGravity {z=-980}
   }
   n3 = "Constraint - dyn_shirt_c_MSTR (P)"(ComponentPose=@n2, Alpha=1, ConstraintWeights_0=1) {
    alphaBoolBlend {blendOption=Linear}
    boneToModify {boneName=dyn_shirt_c_MSTR}
    constraintSetup [{targetBone {boneName=pelvis}}]
    constraintWeights [1]
   }
   n4 = "Transform (Modify) Bone - Bone: dyn_shirt_b_MSTR"!(ComponentPose=@n3, bAlphaBoolEnabled:bool=True, Translation="-10,0,0", Rotation="0,0,0", Scale="1,1,1") -> (Pose) {
    alphaInputType=Bool
    rotationMode=BMM_Additive
    scaleMode=BMM_Additive
    translationMode=BMM_Additive
    alphaBoolBlend {blendInTime=0.2, blendOption=Linear, blendOutTime=0.1}
    boneToModify {boneName=dyn_shirt_b_MSTR}
    translation {x=-10}
   }
   n5 = "Transform (Modify) Bone - Bone: dyn_back_collar_MSTR"!(ComponentPose=@n4, bAlphaBoolEnabled:bool=True, Translation="0,-10,0", Rotation="0,0,-20", Scale="1,1,1") -> (Pose) {
    alphaInputType=Bool
    rotationMode=BMM_Additive
    scaleMode=BMM_Additive
    translationMode=BMM_Additive
    alphaBoolBlend {blendInTime=0.2, blendOption=Linear, blendOutTime=0.1}
    boneToModify {boneName=dyn_back_collar_MSTR}
    rotation {roll=-20}
    translation {y=-10}
   }
   n6 = "Transform (Modify) Bone - Bone: dyn_r_02_collar_MSTR"!(ComponentPose=@n5, bAlphaBoolEnabled:bool=True, Translation="0,0,0", Rotation="0,0,-20", Scale="1,1,1") -> (Pose) {
    alphaInputType=Bool
    rotationMode=BMM_Additive
    scaleMode=BMM_Additive
    translationMode=BMM_Additive
    alphaBoolBlend {blendInTime=0.2, blendOption=Linear, blendOutTime=0.1}
    boneToModify {boneName=dyn_r_02_collar_MSTR}
    rotation {roll=-20}
   }
   n7 = "Transform (Modify) Bone - Bone: dyn_r_02_collar_MSTR"!(ComponentPose=@n6, Alpha:float=Get LookUpRifle() -> (LookUpRifle:double), Translation="0,0,0", Rotation="0,0,-50", Scale="1,1,1") -> (Pose) {
    rotationMode=BMM_Additive
    scaleMode=BMM_Additive
    translationMode=BMM_Additive
    alphaBoolBlend {blendOption=Linear}
    boneToModify {boneName=dyn_r_02_collar_MSTR}
    rotation {roll=-50}
   }
   n8 = "Transform (Modify) Bone - Bone: dyn_back_collar_MSTR"!(ComponentPose=@n7, Alpha:float=Get LookUp() -> (LookUp:double), Translation="0,-10,0", Rotation="0,0,-25", Scale="1,1,1") -> (Pose) {
    rotationMode=BMM_Additive
    scaleMode=BMM_Additive
    translationMode=BMM_Additive
    alphaBoolBlend {blendOption=Linear}
    boneToModify {boneName=dyn_back_collar_MSTR}
    rotation {roll=-25}
    translation {y=-10}
   }
   n9 = Control Rig(Source=Blend Poses by bool(
     BlendPose_0=Component To Local(ComponentPose=@n8)
     BlendPose_1=Copy Pose From Mesh() {bCopyCurves=true, bUseAttachedParent=true}
     bActiveValue=Get bIsCrouching()
    ) {blendType=Linear}) {
    alphaBoolBlend {blendOption=Linear}
    controlRigAssetReference {blueprintRigClass=&/M_MED_Pliant_CTRLRig.M_MED_Pliant_CTRLRig_C}
   }
   n10 = Control Rig(Source=@n9) {
    alphaBoolBlend {blendOption=Linear}
    controlRigAssetReference {blueprintRigClass=&/M_MED_Pliant_Deform_CtrlRig_sculpters.M_MED_Pliant_Deform_CtrlRig_sculpters_C}
   }
   n11 = Control Rig(Source=@n10) {
    alphaBoolBlend {blendOption=Linear}
    controlRigAssetReference {blueprintRigClass=&/M_MED_Pliant_Deform_CtrlRig_twists.M_MED_Pliant_Deform_CtrlRig_twists_C}
   }
   n12 = "Transform (Modify) Bone - Bone: dyn_root_slope"!(
    ComponentPose=Local To Component(LocalPose=@n11)
    bAlphaBoolEnabled:bool=NOT Boolean(A=Get bCanPlayCustomAnimations() -> (bCanPlayCustomAnimations:bool))
    Translation=Get slope_root_trans() -> (slope_root_trans:Vector)
    Rotation=Get slope_foot_rot() -> (slope_foot_rot:Rotator)
    Scale="1,1,1"
   ) -> (Pose) {
    alphaInputType=Bool
    rotationMode=BMM_Replace
    rotationSpace=BCS_WorldSpace
    translationMode=BMM_Replace
    translationSpace=BCS_WorldSpace
    alphaBoolBlend {blendOption=Linear}
    boneToModify {boneName=dyn_root_slope}
   }
   n13 = "Transform (Modify) Bone - Bone: dyn_root_simspace"(ComponentPose=@n12, Alpha=1, Translation="0,0,0", Rotation=Get sim_space_rotation() -> (sim_space_rotation:Rotator), Scale="1,1,1") {
    rotationMode=BMM_Replace
    rotationSpace=BCS_WorldSpace
    translationSpace=BCS_BoneSpace
    alphaBoolBlend {blendOption=Linear}
    alphaScaleBiasClamp {bInterpResult=true, interpSpeedDecreasing=2, interpSpeedIncreasing=2}
    boneToModify {boneName=dyn_root_simspace}
   }
   n14 = "Constraint - dyn_jacket_a_MSTR (P)"(ComponentPose=@n13, Alpha=1, ConstraintWeights_0=1) {
    alphaBoolBlend {blendOption=Linear}
    boneToModify {boneName=dyn_jacket_a_MSTR}
    constraintSetup [{targetBone {boneName=pelvis}}]
    constraintWeights [1]
   }
   n15 = "Constraint - dyn_jacket_b_MSTR (P)"(ComponentPose=@n14, Alpha=1, ConstraintWeights_0=1) {
    alphaBoolBlend {blendOption=Linear}
    boneToModify {boneName=dyn_jacket_b_MSTR}
    constraintSetup [{targetBone {boneName=pelvis}}]
    constraintWeights [1]
   }
   n16 = "Constraint - dyn_jacket_c_MSTR (P)"(ComponentPose=@n15, Alpha=1, ConstraintWeights_0=1) {
    alphaBoolBlend {blendOption=Linear}
    boneToModify {boneName=dyn_jacket_c_MSTR}
    constraintSetup [{targetBone {boneName=pelvis}}]
    constraintWeights [1]
   }
   n17 = "Constraint - dyn_jacket_d_MSTR (P)"(ComponentPose=@n16, Alpha=1, ConstraintWeights_0=1) {
    alphaBoolBlend {blendOption=Linear}
    boneToModify {boneName=dyn_jacket_d_MSTR}
    constraintSetup [{targetBone {boneName=pelvis}}]
    constraintWeights [1]
   }
   n18 = "Constraint - dyn_jacket_e_MSTR (P)"(ComponentPose=@n17, Alpha=1, ConstraintWeights_0=1) {
    alphaBoolBlend {blendOption=Linear}
    boneToModify {boneName=dyn_jacket_e_MSTR}
    constraintSetup [{targetBone {boneName=pelvis}}]
    constraintWeights [1]
   }
   n19 = "Constraint - dyn_jacket_f_MSTR (P)"(ComponentPose=@n18, Alpha=1, ConstraintWeights_0=1) {
    alphaBoolBlend {blendOption=Linear}
    boneToModify {boneName=dyn_jacket_f_MSTR}
    constraintSetup [{targetBone {boneName=pelvis}}]
    constraintWeights [1]
   }
   n20 = "Constraint - dyn_jacket_g_MSTR (P)"(ComponentPose=@n19, Alpha=1, ConstraintWeights_0=1) {
    alphaBoolBlend {blendOption=Linear}
    boneToModify {boneName=dyn_jacket_g_MSTR}
    constraintSetup [{targetBone {boneName=pelvis}}]
    constraintWeights [1]
   }
   n21 = SelectFloat(A=0.5, B=1, bPickA=Get bIsCrouching())
   n22 = Anim Dynamics - Bone: dyn_jacket_a_01(ComponentPose=@n20, Alpha=@n21, ExternalForce="0,0,0") {
    angularDampingOverride=0.9999
    angularSpringConstant=25
    bAngularSpring=true
    bGravityOverrideInSimSpace=true
    bOverrideAngularDamping=true
    simulationSpace=BoneRelative
    alphaBoolBlend {blendOption=Linear}
    boundBone {boneName=dyn_jacket_a_01}
    componentAppliedLinearAccClamp=<3500 3500 3500>
    componentLinearAccScale=<1.5 1.5 1.5>
    componentLinearVelScale=<1.5 1.5 1.5>
    physicsBodyDefinitions [{
      sphereCollisionRadius=0
      boundBone {boneName=dyn_jacket_a_01}
      boxExtents=<10 20 20>
      constraintSetup {angularLimitsMax=<25 0 25>, angularLimitsMin=<-25 -25 -25>, angularTarget=<0 0 0>}
      localJointOffset=<2 0 0>
     }]
    relativeSpaceBone {boneName=dyn_root_simspace}
   }
   n23 = Anim Dynamics - Bone: dyn_jacket_a_02(ComponentPose=@n22, Alpha=@n21, ExternalForce="0,0,0") {
    angularDampingOverride=0.9999
    angularSpringConstant=25
    bAngularSpring=true
    bGravityOverrideInSimSpace=true
    bOverrideAngularDamping=true
    simulationSpace=BoneRelative
    alphaBoolBlend {blendOption=Linear}
    boundBone {boneName=dyn_jacket_a_02}
    componentAppliedLinearAccClamp=<3500 3500 3500>
    componentLinearAccScale=<1.5 1.5 1.5>
    componentLinearVelScale=<1.5 1.5 1.5>
    physicsBodyDefinitions [{
      sphereCollisionRadius=0
      boundBone {boneName=dyn_jacket_a_02}
      boxExtents=<10 20 20>
      constraintSetup {angularLimitsMax=<25 0 25>, angularLimitsMin=<-25 -25 -25>, angularTarget=<0 0 0>}
      localJointOffset=<2 0 0>
     }]
    relativeSpaceBone {boneName=dyn_root_simspace}
   }
   n24 = Anim Dynamics - Bone: dyn_jacket_a_03(ComponentPose=@n23, Alpha=@n21, ExternalForce="0,0,0") {
    angularDampingOverride=0.9999
    angularSpringConstant=25
    bAngularSpring=true
    bGravityOverrideInSimSpace=true
    bOverrideAngularDamping=true
    simulationSpace=BoneRelative
    alphaBoolBlend {blendOption=Linear}
    boundBone {boneName=dyn_jacket_a_03}
    componentAppliedLinearAccClamp=<3500 3500 3500>
    componentLinearAccScale=<1.5 1.5 1.5>
    componentLinearVelScale=<1.5 1.5 1.5>
    physicsBodyDefinitions [{
      sphereCollisionRadius=0
      boundBone {boneName=dyn_jacket_a_03}
      boxExtents=<10 20 20>
      constraintSetup {angularLimitsMax=<25 0 25>, angularLimitsMin=<-25 -25 -25>, angularTarget=<0 0 0>}
      localJointOffset=<2 0 0>
     }]
    relativeSpaceBone {boneName=dyn_root_simspace}
   }
   n25 = Anim Dynamics - Bone: dyn_jacket_b_01(ComponentPose=@n24, Alpha=1, ExternalForce="0,0,0") {
    angularDampingOverride=0.9999
    angularSpringConstant=25
    bAngularSpring=true
    bGravityOverrideInSimSpace=true
    bOverrideAngularDamping=true
    simulationSpace=BoneRelative
    alphaBoolBlend {blendOption=Linear}
    boundBone {boneName=dyn_jacket_b_01}
    componentAppliedLinearAccClamp=<3500 3500 3500>
    componentLinearAccScale=<1.5 1.5 1.5>
    componentLinearVelScale=<1.5 1.5 1.5>
    physicsBodyDefinitions [{
      sphereCollisionRadius=0
      boundBone {boneName=dyn_jacket_b_01}
      boxExtents=<10 20 20>
      constraintSetup {angularLimitsMax=<15 2 15>, angularLimitsMin=<-15 -30 -15>, angularTarget=<0 0 0>}
      localJointOffset=<2 0 0>
     }]
    relativeSpaceBone {boneName=dyn_root_simspace}
   }
   n26 = Anim Dynamics - Bone: dyn_jacket_b_02(ComponentPose=@n25, Alpha=1, ExternalForce="0,0,0") {
    angularDampingOverride=0.9999
    angularSpringConstant=25
    bAngularSpring=true
    bGravityOverrideInSimSpace=true
    bOverrideAngularDamping=true
    simulationSpace=BoneRelative
    alphaBoolBlend {blendOption=Linear}
    boundBone {boneName=dyn_jacket_b_02}
    componentAppliedLinearAccClamp=<3500 3500 3500>
    componentLinearAccScale=<1.5 1.5 1.5>
    componentLinearVelScale=<1.5 1.5 1.5>
    physicsBodyDefinitions [{
      sphereCollisionRadius=0
      boundBone {boneName=dyn_jacket_b_02}
      boxExtents=<10 20 20>
      constraintSetup {angularLimitsMax=<15 2 15>, angularLimitsMin=<-15 -30 -15>, angularTarget=<0 0 0>}
      localJointOffset=<2 0 0>
     }]
    relativeSpaceBone {boneName=dyn_root_simspace}
   }
   n27 = Anim Dynamics - Bone: dyn_jacket_c_01(ComponentPose=@n26, Alpha=1, ExternalForce="0,0,0") {
    angularDampingOverride=0.9999
    angularSpringConstant=25
    bAngularSpring=true
    bGravityOverrideInSimSpace=true
    bOverrideAngularDamping=true
    simulationSpace=BoneRelative
    alphaBoolBlend {blendOption=Linear}
    boundBone {boneName=dyn_jacket_c_01}
    componentAppliedLinearAccClamp=<3500 3500 3500>
    componentLinearAccScale=<1.5 1.5 1.5>
    componentLinearVelScale=<1.5 1.5 1.5>
    physicsBodyDefinitions [{
      sphereCollisionRadius=0
      boundBone {boneName=dyn_jacket_c_01}
      boxExtents=<10 20 20>
      constraintSetup {angularLimitsMax=<15 2 15>, angularLimitsMin=<-15 -30 -15>, angularTarget=<0 0 0>}
      localJointOffset=<2 0 0>
     }]
    relativeSpaceBone {boneName=dyn_root_simspace}
   }
   n28 = Anim Dynamics - Bone: dyn_jacket_c_02(ComponentPose=@n27, Alpha=1, ExternalForce="0,0,0") {
    angularDampingOverride=0.9999
    angularSpringConstant=25
    bAngularSpring=true
    bGravityOverrideInSimSpace=true
    bOverrideAngularDamping=true
    simulationSpace=BoneRelative
    alphaBoolBlend {blendOption=Linear}
    boundBone {boneName=dyn_jacket_c_02}
    componentAppliedLinearAccClamp=<3500 3500 3500>
    componentLinearAccScale=<1.5 1.5 1.5>
    componentLinearVelScale=<1.5 1.5 1.5>
    physicsBodyDefinitions [{
      sphereCollisionRadius=0
      boundBone {boneName=dyn_jacket_c_02}
      boxExtents=<10 20 20>
      constraintSetup {angularLimitsMax=<15 2 15>, angularLimitsMin=<-15 -30 -15>, angularTarget=<0 0 0>}
      localJointOffset=<2 0 0>
     }]
    relativeSpaceBone {boneName=dyn_root_simspace}
   }
   n29 = Anim Dynamics - Bone: dyn_jacket_d_01(ComponentPose=@n28, Alpha=1, ExternalForce="0,0,0") {
    angularDampingOverride=0.9999
    angularSpringConstant=25
    bAngularSpring=true
    bGravityOverrideInSimSpace=true
    bOverrideAngularDamping=true
    simulationSpace=BoneRelative
    alphaBoolBlend {blendOption=Linear}
    boundBone {boneName=dyn_jacket_d_01}
    componentAppliedLinearAccClamp=<3500 3500 3500>
    componentLinearAccScale=<1.5 1.5 1.5>
    componentLinearVelScale=<1.5 1.5 1.5>
    physicsBodyDefinitions [{
      sphereCollisionRadius=0
      boundBone {boneName=dyn_jacket_d_01}
      boxExtents=<10 20 20>
      constraintSetup {angularLimitsMax=<15 -5 15>, angularLimitsMin=<-15 -30 -15>, angularTarget=<0 0 0>}
      localJointOffset=<2 0 0>
     }]
    relativeSpaceBone {boneName=dyn_root_simspace}
   }
   n30 = Anim Dynamics - Bone: dyn_jacket_d_02(ComponentPose=@n29, Alpha=1, ExternalForce="0,0,0") {
    angularDampingOverride=0.9999
    angularSpringConstant=25
    bAngularSpring=true
    bGravityOverrideInSimSpace=true
    bOverrideAngularDamping=true
    simulationSpace=BoneRelative
    alphaBoolBlend {blendOption=Linear}
    boundBone {boneName=dyn_jacket_d_02}
    componentAppliedLinearAccClamp=<3500 3500 3500>
    componentLinearAccScale=<1.5 1.5 1.5>
    componentLinearVelScale=<1.5 1.5 1.5>
    physicsBodyDefinitions [{
      sphereCollisionRadius=0
      boundBone {boneName=dyn_jacket_d_02}
      boxExtents=<10 20 20>
      constraintSetup {angularLimitsMax=<15 -5 15>, angularLimitsMin=<-15 -30 -15>, angularTarget=<0 0 0>}
      localJointOffset=<2 0 0>
     }]
    relativeSpaceBone {boneName=dyn_root_simspace}
   }
   n31 = Anim Dynamics - Bone: dyn_jacket_e_01(ComponentPose=@n30, Alpha=1, ExternalForce="0,0,0") {
    angularDampingOverride=0.9999
    angularSpringConstant=25
    bAngularSpring=true
    bGravityOverrideInSimSpace=true
    bOverrideAngularDamping=true
    simulationSpace=BoneRelative
    alphaBoolBlend {blendOption=Linear}
    boundBone {boneName=dyn_jacket_e_01}
    componentAppliedLinearAccClamp=<3500 3500 3500>
    componentLinearAccScale=<1.5 1.5 1.5>
    componentLinearVelScale=<1.5 1.5 1.5>
    physicsBodyDefinitions [{
      sphereCollisionRadius=0
      boundBone {boneName=dyn_jacket_e_01}
      boxExtents=<10 20 20>
      constraintSetup {angularLimitsMax=<15 -5 15>, angularLimitsMin=<-15 -30 -15>, angularTarget=<0 0 0>}
      localJointOffset=<2 0 0>
     }]
    relativeSpaceBone {boneName=dyn_root_simspace}
   }
   n32 = Anim Dynamics - Bone: dyn_jacket_e_02(ComponentPose=@n31, Alpha=1, ExternalForce="0,0,0") {
    angularDampingOverride=0.9999
    angularSpringConstant=25
    bAngularSpring=true
    bGravityOverrideInSimSpace=true
    bOverrideAngularDamping=true
    simulationSpace=BoneRelative
    alphaBoolBlend {blendOption=Linear}
    boundBone {boneName=dyn_jacket_e_02}
    componentAppliedLinearAccClamp=<3500 3500 3500>
    componentLinearAccScale=<1.5 1.5 1.5>
    componentLinearVelScale=<1.5 1.5 1.5>
    physicsBodyDefinitions [{
      sphereCollisionRadius=0
      boundBone {boneName=dyn_jacket_e_02}
      boxExtents=<10 20 20>
      constraintSetup {angularLimitsMax=<15 -5 15>, angularLimitsMin=<-15 -30 -15>, angularTarget=<0 0 0>}
      localJointOffset=<2 0 0>
     }]
    relativeSpaceBone {boneName=dyn_root_simspace}
   }
   n33 = Anim Dynamics - Bone: dyn_jacket_f_01(ComponentPose=@n32, Alpha=1, ExternalForce="0,0,0") {
    angularDampingOverride=0.9999
    angularSpringConstant=25
    bAngularSpring=true
    bGravityOverrideInSimSpace=true
    bOverrideAngularDamping=true
    simulationSpace=BoneRelative
    alphaBoolBlend {blendOption=Linear}
    boundBone {boneName=dyn_jacket_f_01}
    componentAppliedLinearAccClamp=<3500 3500 3500>
    componentLinearAccScale=<1.5 1.5 1.5>
    componentLinearVelScale=<1.5 1.5 1.5>
    physicsBodyDefinitions [{
      sphereCollisionRadius=0
      boundBone {boneName=dyn_jacket_f_01}
      boxExtents=<10 20 20>
      constraintSetup {angularLimitsMax=<15 -5 15>, angularLimitsMin=<-15 -30 -15>, angularTarget=<0 0 0>}
      localJointOffset=<2 0 0>
     }]
    relativeSpaceBone {boneName=dyn_root_simspace}
   }
   n34 = Anim Dynamics - Bone: dyn_jacket_f_02(ComponentPose=@n33, Alpha=1, ExternalForce="0,0,0") {
    angularDampingOverride=0.9999
    angularSpringConstant=25
    bAngularSpring=true
    bGravityOverrideInSimSpace=true
    bOverrideAngularDamping=true
    simulationSpace=BoneRelative
    alphaBoolBlend {blendOption=Linear}
    boundBone {boneName=dyn_jacket_f_02}
    componentAppliedLinearAccClamp=<3500 3500 3500>
    componentLinearAccScale=<1.5 1.5 1.5>
    componentLinearVelScale=<1.5 1.5 1.5>
    physicsBodyDefinitions [{
      sphereCollisionRadius=0
      boundBone {boneName=dyn_jacket_f_02}
      boxExtents=<10 20 20>
      constraintSetup {angularLimitsMax=<15 -5 15>, angularLimitsMin=<-15 -30 -15>, angularTarget=<0 0 0>}
      localJointOffset=<2 0 0>
     }]
    relativeSpaceBone {boneName=dyn_root_simspace}
   }
   n35 = Anim Dynamics - Bone: dyn_jacket_g_01(ComponentPose=@n34, Alpha=1, ExternalForce="0,0,0") {
    angularDampingOverride=0.9999
    angularSpringConstant=25
    bAngularSpring=true
    bGravityOverrideInSimSpace=true
    bOverrideAngularDamping=true
    simulationSpace=BoneRelative
    alphaBoolBlend {blendOption=Linear}
    boundBone {boneName=dyn_jacket_g_01}
    componentAppliedLinearAccClamp=<3500 3500 3500>
    componentLinearAccScale=<1.5 1.5 1.5>
    componentLinearVelScale=<1.5 1.5 1.5>
    physicsBodyDefinitions [{
      sphereCollisionRadius=0
      boundBone {boneName=dyn_jacket_g_01}
      boxExtents=<10 20 20>
      constraintSetup {angularLimitsMax=<25 0 25>, angularLimitsMin=<-25 -25 -25>, angularTarget=<0 0 0>}
      localJointOffset=<2 0 0>
     }]
    relativeSpaceBone {boneName=dyn_root_simspace}
   }
   n36 = "Transform (Modify) Bone - Bone: dyn_jacket_g_MSTR"(ComponentPose=@n35, Alpha=1, Translation="-2.5,2.5,1", Rotation="0,0,0", Scale="1,1,1") {
    rotationMode=BMM_Additive
    translationMode=BMM_Additive
    alphaBoolBlend {blendOption=Linear}
    alphaScaleBiasClamp {bInterpResult=true}
    boneToModify {boneName=dyn_jacket_g_MSTR}
    translation=<-2.5 2.5 1>
   }
   n37 = "Transform (Modify) Bone - Bone: dyn_jacket_f_MSTR"(ComponentPose=@n36, Alpha=1, Translation="-2,3,1", Rotation="0,0,0", Scale="1,1,1") {
    rotationMode=BMM_Additive
    translationMode=BMM_Additive
    alphaBoolBlend {blendOption=Linear}
    alphaScaleBiasClamp {bInterpResult=true}
    boneToModify {boneName=dyn_jacket_f_MSTR}
    translation=<-2 3 1>
   }
   n38 = "Transform (Modify) Bone - Bone: dyn_jacket_a_MSTR"(ComponentPose=@n37, Alpha=1, Translation="2,2,0", Rotation="0,0,0", Scale="1,1,1") {
    rotationMode=BMM_Additive
    translationMode=BMM_Additive
    alphaBoolBlend {blendOption=Linear}
    alphaScaleBiasClamp {bInterpResult=true}
    boneToModify {boneName=dyn_jacket_a_MSTR}
    translation=<2 2>
   }
   n39 = "Transform (Modify) Bone - Bone: dyn_jacket_c_MSTR"!(ComponentPose=@n38, bAlphaBoolEnabled:bool=True, Translation="0,0,5", Rotation="0,0,0", Scale="1,1,1") -> (Pose) {
    alphaInputType=Bool
    rotationMode=BMM_Additive
    rotationSpace=BCS_BoneSpace
    scaleMode=BMM_Additive
    scaleSpace=BCS_BoneSpace
    translationMode=BMM_Additive
    translationSpace=BCS_BoneSpace
    alphaBoolBlend {blendInTime=0.1, blendOption=Linear, blendOutTime=0.1}
    boneToModify {boneName=dyn_jacket_c_MSTR}
    translation {z=5}
   }
   n40 = "Transform (Modify) Bone - Bone: dyn_jacket_d_MSTR"!(ComponentPose=@n39, bAlphaBoolEnabled:bool=True, Translation="0,0,5", Rotation="0,0,0", Scale="1,1,1") -> (Pose) {
    alphaInputType=Bool
    rotationMode=BMM_Additive
    rotationSpace=BCS_BoneSpace
    scaleMode=BMM_Additive
    scaleSpace=BCS_BoneSpace
    translationMode=BMM_Additive
    translationSpace=BCS_BoneSpace
    alphaBoolBlend {blendInTime=0.1, blendOption=Linear, blendOutTime=0.1}
    boneToModify {boneName=dyn_jacket_d_MSTR}
    translation {z=5}
   }
   n41 = "Transform (Modify) Bone - Bone: dyn_jacket_e_MSTR"!(ComponentPose=@n40, bAlphaBoolEnabled:bool=True, Translation="0,0,5", Rotation="0,0,0", Scale="1,1,1") -> (Pose) {
    alphaInputType=Bool
    rotationMode=BMM_Additive
    rotationSpace=BCS_BoneSpace
    scaleMode=BMM_Additive
    scaleSpace=BCS_BoneSpace
    translationMode=BMM_Additive
    translationSpace=BCS_BoneSpace
    alphaBoolBlend {blendInTime=0.1, blendOption=Linear, blendOutTime=0.1}
    boneToModify {boneName=dyn_jacket_e_MSTR}
    translation {z=5}
   }
   n42 = "Transform (Modify) Bone - Bone: dyn_shirt_d_MSTR"(ComponentPose=@n41, Alpha=1, Translation="-5,0,0", Rotation="0,0,25", Scale="1,1,1") {
    rotationMode=BMM_Additive
    scaleMode=BMM_Additive
    translationMode=BMM_Additive
    alphaBoolBlend {blendOption=Linear}
    alphaScaleBiasClamp {bInterpResult=true, interpSpeedDecreasing=2, interpSpeedIncreasing=2}
    boneToModify {boneName=dyn_shirt_d_MSTR}
    rotation {roll=25}
    translation {x=-5}
   }
   n43 = "Transform (Modify) Bone - Bone: dyn_shirt_d_01"(ComponentPose=@n42, Alpha=1, Translation="0,0,0", Rotation="-50,0,25", Scale="1,1,1") {
    rotationMode=BMM_Additive
    scaleMode=BMM_Additive
    translationMode=BMM_Additive
    alphaBoolBlend {blendOption=Linear}
    alphaScaleBiasClamp {bInterpResult=true, interpSpeedDecreasing=2, interpSpeedIncreasing=2}
    boneToModify {boneName=dyn_shirt_d_01}
    rotation {pitch=-50, roll=25}
   }
   n44 = "Transform (Modify) Bone - Bone: dyn_shirt_d_02"(ComponentPose=@n43, Alpha=1, Translation="0,0,0", Rotation="-50,0,0", Scale="1,1,1") {
    rotationMode=BMM_Additive
    scaleMode=BMM_Additive
    translationMode=BMM_Additive
    alphaBoolBlend {blendOption=Linear}
    alphaScaleBiasClamp {bInterpResult=true, interpSpeedDecreasing=2, interpSpeedIncreasing=2}
    boneToModify {boneName=dyn_shirt_d_02}
    rotation {pitch=-50}
   }
   n45 = "Transform (Modify) Bone - Bone: dyn_shirt_d_03"(ComponentPose=@n44, Alpha=1, Translation="0,0,0", Rotation="-50,0,0", Scale="1,1,1") {
    rotationMode=BMM_Additive
    scaleMode=BMM_Additive
    translationMode=BMM_Additive
    alphaBoolBlend {blendOption=Linear}
    alphaScaleBiasClamp {bInterpResult=true, interpSpeedDecreasing=2, interpSpeedIncreasing=2}
    boneToModify {boneName=dyn_shirt_d_03}
    rotation {pitch=-50}
   }
   n46 = "Transform (Modify) Bone - Bone: dyn_shirt_d_04"(ComponentPose=@n45, Alpha=1, Translation="0,0,0", Rotation="-50,0,0", Scale="1,1,1") {
    rotationMode=BMM_Additive
    scaleMode=BMM_Additive
    translationMode=BMM_Additive
    alphaBoolBlend {blendOption=Linear}
    alphaScaleBiasClamp {bInterpResult=true, interpSpeedDecreasing=2, interpSpeedIncreasing=2}
    boneToModify {boneName=dyn_shirt_d_04}
    rotation {pitch=-50}
   }
   n47 = "Transform (Modify) Bone - Bone: dyn_shirt_c_MSTR"(ComponentPose=@n46, Alpha=1, Translation="0,0,10", Rotation="0,0,0", Scale="1,1,1") {
    rotationMode=BMM_Additive
    rotationSpace=BCS_BoneSpace
    scaleMode=BMM_Additive
    scaleSpace=BCS_BoneSpace
    translationMode=BMM_Replace
    translationSpace=BCS_BoneSpace
    alphaBoolBlend {blendOption=Linear}
    alphaScaleBiasClamp {bInterpResult=true, interpSpeedDecreasing=2, interpSpeedIncreasing=2}
    boneToModify {boneName=dyn_shirt_c_MSTR}
    translation {z=10}
   }
   n48 = Get bIsCrouching()
   n49 = RigidBody!(ComponentPose=@n47, Alpha:float="To Float (Boolean)"(InBool=NOT Boolean(A=@n48)), OverrideWorldGravity=Get GravityOverride()) -> (Pose) {
    bClampLinearTranslationLimitToRefPose=true
    bOverrideWorldGravity=true
    overridePhysicsAsset=&/M_MED_Pliant_Shirt_Physics
    simulationSpace=BaseBoneSpace
    alphaBoolBlend {blendOption=Linear}
    baseBoneRef {boneName=root}
    componentAppliedLinearAccClamp=<1500 1500 1500>
    componentLinearAccScale=<0.5 0.5 0.5>
    overrideWorldGravity {z=-980}
   }
   n50 = RigidBody!(ComponentPose=@n49, bAlphaBoolEnabled:bool, OverrideWorldGravity=Get GravityOverride(), ExternalForce:Vector) -> (Pose) {
    alphaInputType=Bool
    bClampLinearTranslationLimitToRefPose=true
    bOverrideWorldGravity=true
    overridePhysicsAsset=&M_MED_Pliant_Shirt_Crouch_Physics
    simulationSpace=BaseBoneSpace
    alphaBoolBlend {blendInTime=0.1, blendOption=Linear, blendOutTime=0.1}
    baseBoneRef {boneName=root}
    componentAppliedLinearAccClamp=<5000 5000 5000>
    componentLinearAccScale=<0.5 0.5 0.5>
    overrideWorldGravity {z=-980}
   }
   n51 = "Transform (Modify) Bone - Bone: dyn_shirt_d_MSTR"(ComponentPose=@n50, Alpha=1, Translation="0,0,0", Rotation="0,-30,0", Scale="1,1,1") {
    rotationMode=BMM_Additive
    rotationSpace=BCS_BoneSpace
    scaleMode=BMM_Additive
    scaleSpace=BCS_BoneSpace
    translationMode=BMM_Replace
    translationSpace=BCS_BoneSpace
    alphaBoolBlend {blendOption=Linear}
    alphaScaleBiasClamp {bInterpResult=true, interpSpeedDecreasing=2, interpSpeedIncreasing=2}
    boneToModify {boneName=dyn_shirt_d_MSTR}
    rotation {yaw=-30}
   }
   n52 = "Transform (Modify) Bone - Bone: dyn_shirt_d_MSTR"(ComponentPose=@n51, Alpha=1, Translation="0,0,2.5", Rotation="-5,0,0", Scale="1,1,1") {
    rotationMode=BMM_Additive
    rotationSpace=BCS_BoneSpace
    scaleMode=BMM_Additive
    scaleSpace=BCS_BoneSpace
    translationMode=BMM_Additive
    translationSpace=BCS_BoneSpace
    alphaBoolBlend {blendOption=Linear}
    boneToModify {boneName=dyn_shirt_d_MSTR}
    rotation {pitch=-5}
    translation {z=2.5}
   }
   n53 = "Transform (Modify) Bone - Bone: dyn_shirt_c_MSTR"(ComponentPose=@n52, Alpha=1, Translation="0,0,2.5", Rotation="0,0,0", Scale="1,1,1") {
    rotationMode=BMM_Additive
    rotationSpace=BCS_BoneSpace
    scaleMode=BMM_Additive
    scaleSpace=BCS_BoneSpace
    translationMode=BMM_Additive
    translationSpace=BCS_BoneSpace
    alphaBoolBlend {blendInTime=0.1, blendOption=Linear, blendOutTime=0.1}
    boneToModify {boneName=dyn_shirt_c_MSTR}
    translation {z=2.5}
   }
   n54 = "Transform (Modify) Bone - Bone: dyn_shirt_b_MSTR"!(ComponentPose=@n53, Alpha:float=1, Translation="0,0,0", Rotation="0,0,0", Scale="1,1,1") -> (Pose) {
    rotationMode=BMM_Additive
    rotationSpace=BCS_BoneSpace
    scaleMode=BMM_Additive
    scaleSpace=BCS_BoneSpace
    translationMode=BMM_Additive
    translationSpace=BCS_BoneSpace
    alphaBoolBlend {blendInTime=0.1, blendOption=Linear, blendOutTime=0.1}
    boneToModify {boneName=dyn_shirt_b_MSTR}
   }
   n55 = "Transform (Modify) Bone - Bone: dyn_shirt_c_MSTR"!(ComponentPose=@n54, bAlphaBoolEnabled:bool=false, Translation="0,-10,-10", Rotation="0,0,0", Scale="1,1,1") -> (Pose) {
    alphaInputType=Bool
    rotationMode=BMM_Additive
    rotationSpace=BCS_BoneSpace
    scaleMode=BMM_Additive
    scaleSpace=BCS_BoneSpace
    translationMode=BMM_Additive
    translationSpace=BCS_BoneSpace
    alphaBoolBlend {blendInTime=0.1, blendOption=Linear, blendOutTime=0.1}
    boneToModify {boneName=dyn_shirt_c_MSTR}
    translation {y=-10, z=-10}
   }
   n56 = "Constraint - dyn_l_legStrap_MSTR (R)"(ComponentPose=@n55, Alpha=1, ConstraintWeights_0=1) {
    alphaBoolBlend {blendOption=Linear}
    boneToModify {boneName=dyn_l_legStrap_MSTR}
    constraintSetup [{transformType=Rotation, targetBone {boneName=thigh_l}}]
    constraintWeights [1]
   }
   n57 = "Constraint - dyn_l_legStrap_01 (R)"(ComponentPose=@n56, Alpha=1, ConstraintWeights_0=1) {
    alphaBoolBlend {blendOption=Linear}
    boneToModify {boneName=dyn_l_legStrap_01}
    constraintSetup [{transformType=Rotation, targetBone {boneName=thigh_l}}]
    constraintWeights [1]
   }
   n58 = "Constraint - dyn_l_legStrap_02 (R)"(ComponentPose=@n57, Alpha=1, ConstraintWeights_0=1) {
    alphaBoolBlend {blendOption=Linear}
    boneToModify {boneName=dyn_l_legStrap_02}
    constraintSetup [{transformType=Rotation, targetBone {boneName=thigh_l}}]
    constraintWeights [1]
   }
   n59 = "Constraint - dyn_l_legStrap_03 (R)"(ComponentPose=@n58, Alpha=1, ConstraintWeights_0=1) {
    alphaBoolBlend {blendOption=Linear}
    boneToModify {boneName=dyn_l_legStrap_03}
    constraintSetup [{transformType=Rotation, targetBone {boneName=thigh_l}}]
    constraintWeights [1]
   }
   n60 = "Constraint - dyn_l_legStrap_04 (R)"(ComponentPose=@n59, Alpha=1, ConstraintWeights_0=1) {
    alphaBoolBlend {blendOption=Linear}
    boneToModify {boneName=dyn_l_legStrap_04}
    constraintSetup [{transformType=Rotation, targetBone {boneName=thigh_l}}]
    constraintWeights [1]
   }
   n61 = "Constraint - dyn_l_strapHelper (R)"(ComponentPose=@n60, Alpha=1, ConstraintWeights_0=1) {
    alphaBoolBlend {blendOption=Linear}
    boneToModify {boneName=dyn_l_strapHelper}
    constraintSetup [{transformType=Rotation, targetBone {boneName=thigh_twist_01_l}}]
    constraintWeights [1]
   }
   n62 = "Transform (Modify) Bone - Bone: dyn_l_strapHelper"!(ComponentPose=@n61, bAlphaBoolEnabled:bool=Get bIsCrouching(), Translation="11,0,7.5", Rotation="0,-80,0", Scale="1,1,1") -> (Pose) {
    alphaInputType=Bool
    rotationMode=BMM_Additive
    scaleMode=BMM_Additive
    translationMode=BMM_Additive
    alphaBoolBlend {blendOption=Linear, blendOutTime=0.3}
    boneToModify {boneName=dyn_l_strapHelper}
    rotation {yaw=-80}
    translation {x=11, z=7.5}
   }
   n63 = "Transform (Modify) Bone - Bone: dyn_l_strapHelper"(ComponentPose=@n62, bAlphaBoolEnabled=True, Translation="-3,0,0", Rotation="0,0,0", Scale="1,1,1") {
    alphaInputType=Bool
    rotationMode=BMM_Additive
    scaleMode=BMM_Additive
    translationMode=BMM_Additive
    alphaBoolBlend {blendOption=Linear}
    boneToModify {boneName=dyn_l_strapHelper}
    translation {x=-3}
   }
   n64 = "Transform (Modify) Bone - Bone: dyn_l_strapHelper"(ComponentPose=@n63, bAlphaBoolEnabled=NOT Boolean(A=Get bIsCrouching()), Translation="0,0,0", Rotation="5,0,0", Scale="1,1,1") {
    alphaInputType=Bool
    rotationMode=BMM_Additive
    scaleMode=BMM_Additive
    translationMode=BMM_Additive
    alphaBoolBlend {blendOption=Linear, blendOutTime=0.3}
    boneToModify {boneName=dyn_l_strapHelper}
    rotation {pitch=5}
   }
   n65 = "Transform (Modify) Bone - Bone: dyn_l_legStrap_01"!(ComponentPose=@n64, bAlphaBoolEnabled:bool=True, Translation="0,0,0", Rotation="10,0,0", Scale="1,1,1") -> (Pose) {
    alphaInputType=Bool
    rotationMode=BMM_Additive
    scaleMode=BMM_Additive
    translationMode=BMM_Additive
    alphaBoolBlend {blendOption=Linear, blendOutTime=0.3}
    boneToModify {boneName=dyn_l_legStrap_01}
    rotation {pitch=10}
   }
   n66 = Anim Dynamics - Bone: dyn_l_legStrap_01!(ComponentPose=@n65, bAlphaBoolEnabled:bool=Get bIsCrouching(), ExternalForce="0,0,0") -> (Pose) {
    alphaInputType=Bool
    angularDampingOverride=0.9999
    bAngularSpring=true
    bGravityOverrideInSimSpace=true
    bOverrideAngularDamping=true
    gravityScale=1.2
    simulationSpace=RootRelative
    alphaBoolBlend {blendInTime=0.1, blendOption=Linear, blendOutTime=0.1}
    boundBone {boneName=dyn_l_legStrap_01}
    componentAppliedLinearAccClamp=<3500 3500 4300>
    componentLinearAccScale=<3 3 3>
    componentLinearVelScale=<3 3 3>
    physicsBodyDefinitions [{
      sphereCollisionRadius=0
      boundBone {boneName=dyn_l_legStrap_01}
      boxExtents=<9 18 18>
      constraintSetup {angularLimitsMax=<45 0 60>, angularLimitsMin=<-45 -80 -60>, angularTarget=<0 0 0>}
      localJointOffset=<2 0 0>
     }]
    relativeSpaceBone {boneName=root}
   }
   n67 = Anim Dynamics - Bone: dyn_l_legStrap_01!(ComponentPose=@n66, Alpha:float=SelectFloat(A=0.25, B=0.05, bPickA=Get bIsInFrontEnd()), ExternalForce="0,0,0") -> (Pose) {
    angularDampingOverride=0.999
    bAngularSpring=true
    bGravityOverrideInSimSpace=true
    bOverrideAngularDamping=true
    gravityScale=1.2
    simulationSpace=RootRelative
    alphaBoolBlend {blendInTime=0.1, blendOption=Linear, blendOutTime=0.1}
    boundBone {boneName=dyn_l_legStrap_01}
    componentAppliedLinearAccClamp=<3500 3500 4300>
    componentLinearAccScale=<3 3 3>
    componentLinearVelScale=<3 3 3>
    physicsBodyDefinitions [{
      sphereCollisionRadius=0
      boundBone {boneName=dyn_l_legStrap_01}
      boxExtents=<9 18 18>
      constraintSetup {angularLimitsMax=<45 45 60>, angularLimitsMin=<-45 -80 -60>, angularTarget=<0 0 0>}
      localJointOffset=<2 0 0>
     }]
    relativeSpaceBone {boneName=root}
   }
   n68 = Anim Dynamics - Bone: dyn_l_legStrap_02(ComponentPose=@n67, Alpha=0.8, ExternalForce="0,0,0") {
    angularDampingOverride=0.999
    bAngularSpring=true
    bGravityOverrideInSimSpace=true
    bOverrideAngularDamping=true
    gravityScale=1.2
    simulationSpace=RootRelative
    alphaBoolBlend {blendOption=Linear}
    boundBone {boneName=dyn_l_legStrap_02}
    componentAppliedLinearAccClamp=<3500 3500 4300>
    componentLinearAccScale=<3 3 3>
    componentLinearVelScale=<3 3 3>
    physicsBodyDefinitions [{
      sphereCollisionRadius=0
      boundBone {boneName=dyn_l_legStrap_02}
      boxExtents=<9 18 18>
      constraintSetup {angularLimitsMax=<45 5 60>, angularLimitsMin=<-45 -80 -60>, angularTarget=<0 0 0>}
      localJointOffset=<2 0 0>
     }]
    relativeSpaceBone {boneName=root}
   }
   n69 = Anim Dynamics - Bone: dyn_l_legStrap_03(ComponentPose=@n68, Alpha=1, ExternalForce="0,0,0") {
    angularDampingOverride=0.999
    bAngularSpring=true
    bGravityOverrideInSimSpace=true
    bOverrideAngularDamping=true
    gravityScale=1.2
    simulationSpace=RootRelative
    alphaBoolBlend {blendOption=Linear}
    boundBone {boneName=dyn_l_legStrap_03}
    componentAppliedLinearAccClamp=<3500 3500 4300>
    componentLinearAccScale=<3 3 3>
    componentLinearVelScale=<3 3 3>
    physicsBodyDefinitions [{
      sphereCollisionRadius=0
      boundBone {boneName=dyn_l_legStrap_03}
      boxExtents=<9 18 18>
      constraintSetup {angularLimitsMax=<45 10 60>, angularLimitsMin=<-45 -80 -60>, angularTarget=<0 0 0>}
      localJointOffset=<2 0 0>
     }]
    relativeSpaceBone {boneName=root}
   }
   n70 = Anim Dynamics - Bone: dyn_l_legStrap_04(ComponentPose=@n69, Alpha=1, ExternalForce="0,0,0") {
    angularDampingOverride=0.999
    bAngularSpring=true
    bGravityOverrideInSimSpace=true
    bOverrideAngularDamping=true
    gravityScale=1.2
    simulationSpace=RootRelative
    alphaBoolBlend {blendOption=Linear}
    boundBone {boneName=dyn_l_legStrap_04}
    componentAppliedLinearAccClamp=<3500 3500 4300>
    componentLinearAccScale=<3 3 3>
    componentLinearVelScale=<3 3 3>
    physicsBodyDefinitions [{
      sphereCollisionRadius=0
      boundBone {boneName=dyn_l_legStrap_04}
      boxExtents=<9 18 18>
      constraintSetup {angularLimitsMax=<45 10 60>, angularLimitsMin=<-45 -80 -60>, angularTarget=<0 0 0>}
      localJointOffset=<2 0 0>
     }]
    relativeSpaceBone {boneName=root}
   }
   n71 = "Transform (Modify) Bone - Bone: dyn_l_legStrap_MSTR"!(ComponentPose=@n70, bAlphaBoolEnabled:bool=True, Translation="0,0,0", Rotation="25,0.000,0.000", Scale="1,1,1") -> (Pose) {
    alphaInputType=Bool
    rotationMode=BMM_Additive
    rotationSpace=BCS_BoneSpace
    scaleMode=BMM_Additive
    scaleSpace=BCS_BoneSpace
    translationMode=BMM_Additive
    translationSpace=BCS_BoneSpace
    alphaBoolBlend {blendInTime=0.1, blendOption=Linear, blendOutTime=0.1}
    boneToModify {boneName=dyn_l_legStrap_MSTR}
    rotation {pitch=25}
   }
   n72 = "Transform (Modify) Bone - Bone: dyn_l_legStrap_MSTR"(ComponentPose=@n71, bAlphaBoolEnabled=True, Translation="0,0,0", Rotation="-25,0.000,0.000", Scale="1,1,1") {
    alphaInputType=Bool
    rotationMode=BMM_Additive
    rotationSpace=BCS_BoneSpace
    scaleMode=BMM_Additive
    scaleSpace=BCS_BoneSpace
    translationMode=BMM_Additive
    translationSpace=BCS_BoneSpace
    alphaBoolBlend {blendInTime=0.1, blendOption=Linear, blendOutTime=0.1}
    boneToModify {boneName=dyn_l_legStrap_MSTR}
    rotation {pitch=-25}
   }
   n73 = "Transform (Modify) Bone - Bone: dyn_l_legStrap_MSTR"(ComponentPose=@n72, bAlphaBoolEnabled=True, Translation="0,0,1", Rotation="2.5,0.000,0.000", Scale="1,1,1") {
    alphaInputType=Bool
    rotationMode=BMM_Additive
    rotationSpace=BCS_BoneSpace
    scaleMode=BMM_Additive
    scaleSpace=BCS_BoneSpace
    translationMode=BMM_Additive
    translationSpace=BCS_BoneSpace
    alphaBoolBlend {blendInTime=0.1, blendOption=Linear, blendOutTime=0.1}
    boneToModify {boneName=dyn_l_legStrap_MSTR}
    rotation {pitch=2.5}
    translation {z=1}
   }
   n74 = Anim Dynamics - Bone: dyn_r_legStrap_03(ComponentPose=@n73, Alpha=0.5, ExternalForce="0,0,0") {
    angularDampingOverride=0.9999
    angularSpringConstant=25
    bAngularSpring=true
    bGravityOverrideInSimSpace=true
    bOverrideAngularDamping=true
    simulationSpace=RootRelative
    alphaBoolBlend {blendOption=Linear}
    boundBone {boneName=dyn_r_legStrap_03}
    componentAppliedLinearAccClamp=<3500 3500 3500>
    componentLinearAccScale=<3 3 3>
    componentLinearVelScale=<3 3 3>
    physicsBodyDefinitions [{
      sphereCollisionRadius=0
      boundBone {boneName=dyn_r_legStrap_03}
      boxExtents=<10 20 20>
      constraintSetup {angularLimitsMax=<0 0 15>, angularLimitsMin=<0 0 -15>, angularTarget=<0 0 0>}
      localJointOffset=<2 0 -0.25>
     }]
   }
   n75 = Anim Dynamics - Bone: dyn_r_legStrap_04(ComponentPose=@n74, Alpha=0.5, ExternalForce="0,0,0") {
    angularDampingOverride=0.9999
    angularSpringConstant=25
    bAngularSpring=true
    bGravityOverrideInSimSpace=true
    bOverrideAngularDamping=true
    simulationSpace=RootRelative
    alphaBoolBlend {blendOption=Linear}
    boundBone {boneName=dyn_r_legStrap_04}
    componentAppliedLinearAccClamp=<3500 3500 3500>
    componentLinearAccScale=<3 3 3>
    componentLinearVelScale=<3 3 3>
    physicsBodyDefinitions [{
      sphereCollisionRadius=0
      boundBone {boneName=dyn_r_legStrap_04}
      boxExtents=<10 20 20>
      constraintSetup {angularLimitsMax=<0 0 15>, angularLimitsMin=<0 0 -15>, angularTarget=<0 0 0>}
      localJointOffset=<2 0 -0.25>
     }]
   }
   n76 = Anim Dynamics - Bone: dyn_l_collar(ComponentPose=@n75, Alpha=1, ExternalForce="0,0,0") {
    angularDampingOverride=1
    angularSpringConstant=2500
    bAngularSpring=true
    bGravityOverrideInSimSpace=true
    bOverrideAngularDamping=true
    gravityScale=0.35
    simulationSpace=RootRelative
    alphaBoolBlend {blendOption=Linear}
    boundBone {boneName=dyn_l_collar}
    componentAppliedLinearAccClamp=<2500 2500 2500>
    componentLinearAccScale=<0.6 0.6 0.6>
    componentLinearVelScale=<0.6 0.6 0.6>
    physicsBodyDefinitions [{
      sphereCollisionRadius=0
      boundBone {boneName=dyn_l_collar}
      boxExtents=<10 20 20>
      constraintSetup {angularLimitsMax=<0 2 15>, angularLimitsMin=<0 -2 -15>, angularTarget=<1 0 0>}
      localJointOffset=<1 0 0>
     }]
   }
   n77 = Anim Dynamics - Bone: dyn_r_collar(ComponentPose=@n76, Alpha=1, ExternalForce="0,0,0") {
    angularDampingOverride=1
    angularSpringConstant=2500
    bAngularSpring=true
    bGravityOverrideInSimSpace=true
    bOverrideAngularDamping=true
    gravityScale=0.35
    simulationSpace=RootRelative
    alphaBoolBlend {blendOption=Linear}
    boundBone {boneName=dyn_r_collar}
    componentAppliedLinearAccClamp=<2500 2500 2500>
    componentLinearAccScale=<0.6 0.6 0.6>
    componentLinearVelScale=<0.6 0.6 0.6>
    physicsBodyDefinitions [{
      sphereCollisionRadius=0
      boundBone {boneName=dyn_r_collar}
      boxExtents=<10 20 20>
      constraintSetup {angularLimitsMax=<0 2 15>, angularLimitsMin=<0 -2 -15>, angularTarget=<1 0 0>}
      localJointOffset=<1 0 0>
     }]
   }
   n78 = Anim Dynamics - Bone: dyn_r_01_collar(ComponentPose=@n77, Alpha=1, ExternalForce="0,0,0") {
    angularDampingOverride=1
    angularSpringConstant=2500
    bAngularSpring=true
    bGravityOverrideInSimSpace=true
    bOverrideAngularDamping=true
    gravityScale=0.35
    simulationSpace=RootRelative
    alphaBoolBlend {blendOption=Linear}
    boundBone {boneName=dyn_r_01_collar}
    componentAppliedLinearAccClamp=<2500 2500 2500>
    componentLinearAccScale=<0.6 0.6 0.6>
    componentLinearVelScale=<0.6 0.6 0.6>
    physicsBodyDefinitions [{
      sphereCollisionRadius=0
      boundBone {boneName=dyn_r_01_collar}
      boxExtents=<10 20 20>
      constraintSetup {angularLimitsMax=<0 2 15>, angularLimitsMin=<0 -2 -15>, angularTarget=<1 0 0>}
      localJointOffset=<1 0 0>
     }]
   }
   n79 = Anim Dynamics - Bone: dyn_r_02_collar(ComponentPose=@n78, Alpha=1, ExternalForce="0,0,0") {
    angularDampingOverride=1
    angularSpringConstant=2500
    bAngularSpring=true
    bGravityOverrideInSimSpace=true
    bOverrideAngularDamping=true
    gravityScale=0.35
    simulationSpace=RootRelative
    alphaBoolBlend {blendOption=Linear}
    boundBone {boneName=dyn_r_02_collar}
    componentAppliedLinearAccClamp=<2500 2500 2500>
    componentLinearAccScale=<0.6 0.6 0.6>
    componentLinearVelScale=<0.6 0.6 0.6>
    physicsBodyDefinitions [{
      sphereCollisionRadius=0
      boundBone {boneName=dyn_r_02_collar}
      boxExtents=<10 20 20>
      constraintSetup {angularLimitsMax=<0 2 15>, angularLimitsMin=<0 -2 -15>, angularTarget=<1 0 0>}
      localJointOffset=<1 0 0>
     }]
   }
   n80 = Anim Dynamics - Bone: dyn_back_collar(ComponentPose=@n79, Alpha=1, ExternalForce="0,0,0") {
    angularDampingOverride=1
    angularSpringConstant=2500
    bAngularSpring=true
    bGravityOverrideInSimSpace=true
    bOverrideAngularDamping=true
    gravityScale=0.35
    simulationSpace=RootRelative
    alphaBoolBlend {blendOption=Linear}
    boundBone {boneName=dyn_back_collar}
    componentAppliedLinearAccClamp=<2500 2500 2500>
    componentLinearAccScale=<0.6 0.6 0.6>
    componentLinearVelScale=<0.6 0.6 0.6>
    physicsBodyDefinitions [{
      sphereCollisionRadius=0
      boundBone {boneName=dyn_back_collar}
      boxExtents=<10 20 20>
      constraintSetup {angularLimitsMax=<0 -1 15>, angularLimitsMin=<0 -10 -15>, angularTarget=<1 0 0>}
      localJointOffset=<1 0 0>
     }]
   }
   n81 = Anim Dynamics - Bone: dyn_l_02_collar(ComponentPose=@n80, Alpha=1, ExternalForce="0,0,0") {
    angularDampingOverride=1
    angularSpringConstant=2500
    bAngularSpring=true
    bGravityOverrideInSimSpace=true
    bOverrideAngularDamping=true
    gravityScale=0.35
    simulationSpace=RootRelative
    alphaBoolBlend {blendOption=Linear}
    boundBone {boneName=dyn_l_02_collar}
    componentAppliedLinearAccClamp=<2500 2500 2500>
    componentLinearAccScale=<0.6 0.6 0.6>
    componentLinearVelScale=<0.6 0.6 0.6>
    physicsBodyDefinitions [{
      sphereCollisionRadius=0
      boundBone {boneName=dyn_l_02_collar}
      boxExtents=<10 20 20>
      constraintSetup {angularLimitsMax=<0 2 15>, angularLimitsMin=<0 -2 -15>, angularTarget=<1 0 0>}
      localJointOffset=<1 0 0>
     }]
   }
   n82 = Anim Dynamics - Bone: dyn_l_01_collar(ComponentPose=@n81, Alpha=1, ExternalForce="0,0,0") {
    angularDampingOverride=1
    angularSpringConstant=2500
    bAngularSpring=true
    bGravityOverrideInSimSpace=true
    bOverrideAngularDamping=true
    gravityScale=0.35
    simulationSpace=RootRelative
    alphaBoolBlend {blendOption=Linear}
    boundBone {boneName=dyn_l_01_collar}
    componentAppliedLinearAccClamp=<2500 2500 2500>
    componentLinearAccScale=<0.6 0.6 0.6>
    componentLinearVelScale=<0.6 0.6 0.6>
    physicsBodyDefinitions [{
      sphereCollisionRadius=0
      boundBone {boneName=dyn_l_01_collar}
      boxExtents=<10 20 20>
      constraintSetup {angularLimitsMax=<0 2 15>, angularLimitsMin=<0 -2 -15>, angularTarget=<1 0 0>}
      localJointOffset=<1 0 0>
     }]
   }
   n83 = Get crouch alpha() -> (crouch alpha:double)
   n84 = "Constraint - dyn_jacket_a_MSTR (P)"(ComponentPose=@n82, Alpha=@n83, ConstraintWeights_0=1) {
    alphaBoolBlend {blendOption=Linear}
    boneToModify {boneName=dyn_jacket_a_MSTR}
    constraintSetup [{targetBone {boneName=pelvis}}]
    constraintWeights [1]
   }
   n85 = "Constraint - dyn_jacket_a_01 (P)"(ComponentPose=@n84, Alpha=@n83, ConstraintWeights_0=1) {
    alphaBoolBlend {blendOption=Linear}
    boneToModify {boneName=dyn_jacket_a_01}
    constraintSetup [{targetBone {boneName=pelvis}}]
    constraintWeights [1]
   }
   n86 = "Constraint - dyn_jacket_a_02 (P)"(ComponentPose=@n85, Alpha=@n83, ConstraintWeights_0=1) {
    alphaBoolBlend {blendOption=Linear}
    boneToModify {boneName=dyn_jacket_a_02}
    constraintSetup [{targetBone {boneName=pelvis}}]
    constraintWeights [1]
   }
   n87 = "Constraint - dyn_jacket_a_03 (P)"(ComponentPose=@n86, Alpha=@n83, ConstraintWeights_0=1) {
    alphaBoolBlend {blendOption=Linear}
    boneToModify {boneName=dyn_jacket_a_03}
    constraintSetup [{targetBone {boneName=pelvis}}]
    constraintWeights [1]
   }
   n88 = "Transform (Modify) Bone - Bone: dyn_jacket_a_MSTR"!(ComponentPose=@n87, bAlphaBoolEnabled:bool=True, Translation="5,2,0", Rotation="10.586,-40.335,-30.809", Scale="1,1,1") -> (Pose) {
    alphaInputType=Bool
    rotationMode=BMM_Additive
    scaleMode=BMM_Additive
    translationMode=BMM_Additive
    alphaBoolBlend {blendInTime=0.1, blendOption=Linear, blendOutTime=0.1}
    boneToModify {boneName=dyn_jacket_a_MSTR}
    rotation=<rot 10.586 -40.335 -30.809>
    translation=<5 2>
   }
   n89 = "Transform (Modify) Bone - Bone: dyn_jacket_a_02"!(ComponentPose=@n88, bAlphaBoolEnabled:bool=True, Translation="0,0,0", Rotation="0,0,0", Scale="1,1,1") -> (Pose) {
    alphaInputType=Bool
    rotationMode=BMM_Additive
    scaleMode=BMM_Additive
    translationMode=BMM_Additive
    alphaBoolBlend {blendInTime=0.1, blendOption=Linear, blendOutTime=0.1}
    boneToModify {boneName=dyn_jacket_a_02}
   }
   n90 = "Transform (Modify) Bone - Bone: dyn_jacket_a_03"!(ComponentPose=@n89, bAlphaBoolEnabled:bool=True, Translation="0,0,0", Rotation="0,0,0", Scale="1,1,1") -> (Pose) {
    alphaInputType=Bool
    rotationMode=BMM_Additive
    scaleMode=BMM_Additive
    translationMode=BMM_Additive
    alphaBoolBlend {blendInTime=0.1, blendOption=Linear, blendOutTime=0.1}
    boneToModify {boneName=dyn_jacket_a_03}
   }
   n91 = "Transform (Modify) Bone - Bone: dyn_jacket_b_MSTR"!(ComponentPose=@n90, bAlphaBoolEnabled:bool=True, Translation="5,0,0", Rotation="0,0,15", Scale="1,1,1") -> (Pose) {
    alphaInputType=Bool
    rotationMode=BMM_Additive
    scaleMode=BMM_Additive
    translationMode=BMM_Additive
    alphaBoolBlend {blendInTime=0.1, blendOption=Linear, blendOutTime=0.1}
    boneToModify {boneName=dyn_jacket_b_MSTR}
    rotation {roll=15}
    translation {x=5}
   }
   n92 = "Transform (Modify) Bone - Bone: dyn_jacket_a_02"!(ComponentPose=@n91, Alpha:float=Get LookDown() -> (LookDown:double), Translation="0,0,0", Rotation="20,0,0", Scale="1,1,1") -> (Pose) {
    rotationMode=BMM_Additive
    scaleMode=BMM_Additive
    translationMode=BMM_Additive
    alphaBoolBlend {blendInTime=0.1, blendOption=Linear, blendOutTime=0.1}
    boneToModify {boneName=dyn_jacket_a_02}
    rotation {pitch=20}
   }
   n93 = "Transform (Modify) Bone - Bone: dyn_jacket_g_MSTR"(ComponentPose=@n92, Alpha=1, Translation="0,0,1.5", Rotation="0,0,0", Scale="1,1,1") {
    rotationMode=BMM_Additive
    rotationSpace=BCS_BoneSpace
    scaleMode=BMM_Additive
    scaleSpace=BCS_BoneSpace
    translationMode=BMM_Additive
    translationSpace=BCS_BoneSpace
    alphaBoolBlend {blendInTime=0.1, blendOption=Linear, blendOutTime=0.1}
    boneToModify {boneName=dyn_jacket_g_MSTR}
    translation {z=1.5}
   }
   n94 = "Transform (Modify) Bone - Bone: dyn_r_legStrap_03"(ComponentPose=@n93, Alpha=1, Translation="0,0,0", Rotation="7,0,0", Scale="1,1,1") {
    rotationMode=BMM_Additive
    rotationSpace=BCS_BoneSpace
    scaleMode=BMM_Additive
    scaleSpace=BCS_BoneSpace
    translationMode=BMM_Additive
    translationSpace=BCS_BoneSpace
    alphaBoolBlend {blendOption=Linear}
    boneToModify {boneName=dyn_r_legStrap_03}
    rotation {pitch=7}
   }
   n95 = "Transform (Modify) Bone - Bone: dyn_r_collar_MSTR"(ComponentPose=@n94, Alpha=Get Lookup_Collar() -> (Lookup_Collar:double), Translation="0,0,0", Rotation="50,0.000,0.000", Scale="1,1,1") {
    rotationMode=BMM_Additive
    rotationSpace=BCS_BoneSpace
    scaleMode=BMM_Additive
    scaleSpace=BCS_BoneSpace
    translationMode=BMM_Additive
    translationSpace=BCS_BoneSpace
    alphaBoolBlend {blendOption=Linear}
    boneToModify {boneName=dyn_r_collar_MSTR}
    rotation {pitch=50}
   }
   n96 = "Transform (Modify) Bone - Bone: dyn_r_02_collar_MSTR"(ComponentPose=@n95, Alpha=Get Lookup_Collar(), Translation="0,0,0", Rotation="10,20,0", Scale="1,1,1") {
    rotationMode=BMM_Additive
    rotationSpace=BCS_BoneSpace
    scaleMode=BMM_Additive
    scaleSpace=BCS_BoneSpace
    translationMode=BMM_Additive
    translationSpace=BCS_BoneSpace
    alphaBoolBlend {blendOption=Linear}
    boneToModify {boneName=dyn_r_02_collar_MSTR}
    rotation {pitch=10, yaw=20}
   }
   n97 = "Transform (Modify) Bone - Bone: dyn_r_01_collar_MSTR"(ComponentPose=@n96, Alpha=Get Lookup_Collar(), Translation="0,-3,-5", Rotation="35,30,0", Scale="1,1,1") {
    rotationMode=BMM_Additive
    rotationSpace=BCS_BoneSpace
    scaleMode=BMM_Additive
    scaleSpace=BCS_BoneSpace
    translationMode=BMM_Additive
    translationSpace=BCS_BoneSpace
    alphaBoolBlend {blendOption=Linear}
    boneToModify {boneName=dyn_r_01_collar_MSTR}
    rotation {pitch=35, yaw=30}
    translation {y=-3, z=-5}
   }
   n98 = "Constraint - dyn_r_01_collar_MSTR (P)"(ComponentPose=@n97, Alpha=0.25, ConstraintWeights_0=1) {
    alphaBoolBlend {blendOption=Linear}
    boneToModify {boneName=dyn_r_01_collar_MSTR}
    constraintSetup [{targetBone {boneName=clavicle_r}}]
    constraintWeights [1]
   }
   n99 = "Constraint - dyn_l_01_collar_MSTR (P)"(ComponentPose=@n98, Alpha=0.25, ConstraintWeights_0=1) {
    alphaBoolBlend {blendOption=Linear}
    boneToModify {boneName=dyn_l_01_collar_MSTR}
    constraintSetup [{targetBone {boneName=clavicle_l}}]
    constraintWeights [1]
   }
   n100 = "Transform (Modify) Bone - Bone: dyn_neckFix"(ComponentPose=@n99, Alpha=1, Translation="0,-5,-1.8", Rotation="0,0,0", Scale="1,1,1") {
    rotationMode=BMM_Additive
    rotationSpace=BCS_BoneSpace
    scaleMode=BMM_Additive
    scaleSpace=BCS_BoneSpace
    translationMode=BMM_Additive
    translationSpace=BCS_BoneSpace
    alphaBoolBlend {blendOption=Linear}
    boneToModify {boneName=dyn_neckFix}
    translation {y=-5, z=-1.8}
   }
   n101 = "Transform (Modify) Bone - Bone: dyn_l_unwantedInnerGeo"(ComponentPose=@n100, Alpha=1, Translation="0,-5,0", Rotation="0,0,0", Scale="0.5,0.5,0.5") {
    rotationMode=BMM_Additive
    rotationSpace=BCS_BoneSpace
    scaleMode=BMM_Additive
    scaleSpace=BCS_BoneSpace
    translationMode=BMM_Additive
    translationSpace=BCS_BoneSpace
    alphaBoolBlend {blendOption=Linear}
    boneToModify {boneName=dyn_l_unwantedInnerGeo}
    scale=<0.5 0.5 0.5>
    translation {y=-5}
   }
   n102 = "Transform (Modify) Bone - Bone: dyn_r_unwantedInnerGeo"(ComponentPose=@n101, Alpha=1, Translation="0,-5,0", Rotation="0,0,0", Scale="0.01,0.01,0.01") {
    rotationMode=BMM_Additive
    rotationSpace=BCS_BoneSpace
    scaleMode=BMM_Additive
    scaleSpace=BCS_BoneSpace
    translationMode=BMM_Additive
    translationSpace=BCS_BoneSpace
    alphaBoolBlend {blendOption=Linear}
    boneToModify {boneName=dyn_r_unwantedInnerGeo}
    scale=<0.01 0.01 0.01>
    translation {y=-5}
   }
   n103 = "Transform (Modify) Bone - Bone: dyn_shrinkBack"(ComponentPose=@n102, Alpha=1, Translation="0,0,0", Rotation="0,0,0", Scale="0.25,0.25,0.25") {
    rotationMode=BMM_Additive
    rotationSpace=BCS_BoneSpace
    scaleMode=BMM_Additive
    scaleSpace=BCS_BoneSpace
    translationMode=BMM_Additive
    translationSpace=BCS_BoneSpace
    alphaBoolBlend {blendOption=Linear}
    boneToModify {boneName=dyn_shrinkBack}
    scale=<0.25 0.25 0.25>
   }
   n104 = "Transform (Modify) Bone - Bone: dyn_neckFix_01"(ComponentPose=@n103, Alpha=1, Translation="0,0,3", Rotation="0,0,0", Scale="1,1,1") {
    rotationMode=BMM_Additive
    scaleMode=BMM_Additive
    translationMode=BMM_Additive
    alphaBoolBlend {blendOption=Linear}
    boneToModify {boneName=dyn_neckFix_01}
    translation {z=3}
   }
   n105 = "Transform (Modify) Bone - Bone: dyn_jacket_a_MSTR"(ComponentPose=@n104, bAlphaBoolEnabled=True, Translation="0,0,3", Rotation="0,0,0", Scale="1,1,1") {
    alphaInputType=Bool
    rotationMode=BMM_Additive
    rotationSpace=BCS_BoneSpace
    scaleMode=BMM_Additive
    scaleSpace=BCS_BoneSpace
    translationMode=BMM_Additive
    translationSpace=BCS_BoneSpace
    alphaBoolBlend {blendInTime=0.15, blendOption=Linear}
    boneToModify {boneName=dyn_jacket_a_MSTR}
    translation {z=3}
   }
   n106 = "Transform (Modify) Bone - Bone: dyn_jacket_g_MSTR"!(ComponentPose=@n105, bAlphaBoolEnabled:bool=True, Translation="0,0,3", Rotation="0,0,0", Scale="1,1,1") -> (Pose) {
    alphaInputType=Bool
    rotationMode=BMM_Additive
    rotationSpace=BCS_BoneSpace
    scaleMode=BMM_Additive
    scaleSpace=BCS_BoneSpace
    translationMode=BMM_Additive
    translationSpace=BCS_BoneSpace
    alphaBoolBlend {blendInTime=0.15, blendOption=Linear}
    boneToModify {boneName=dyn_jacket_g_MSTR}
    translation {z=3}
   }
   n107 = "Constraint - dyn_l_collar_MSTR (P)"(ComponentPose=@n106, Alpha=1, ConstraintWeights_0=1) {
    alphaBoolBlend {blendOption=Linear}
    boneToModify {boneName=dyn_l_collar_MSTR}
    constraintSetup [{targetBone {boneName=neck_01}}]
    constraintWeights [1]
   }
   n108 = "Constraint - dyn_r_02_collar_MSTR (P)"(ComponentPose=@n107, Alpha=1, ConstraintWeights_0=1) {
    alphaBoolBlend {blendOption=Linear}
    boneToModify {boneName=dyn_r_02_collar_MSTR}
    constraintSetup [{targetBone {boneName=neck_01}}]
    constraintWeights [1]
   }
   n109 = "Constraint - dyn_back_collar_MSTR (P)"(ComponentPose=@n108, Alpha=1, ConstraintWeights_0=1) {
    alphaBoolBlend {blendOption=Linear}
    boneToModify {boneName=dyn_back_collar_MSTR}
    constraintSetup [{targetBone {boneName=neck_01}}]
    constraintWeights [1]
   }
   n110 = "Constraint - dyn_l_02_collar_MSTR (P)"(ComponentPose=@n109, Alpha=1, ConstraintWeights_0=1) {
    alphaBoolBlend {blendOption=Linear}
    boneToModify {boneName=dyn_l_02_collar_MSTR}
    constraintSetup [{targetBone {boneName=neck_01}}]
    constraintWeights [1]
   }
   n111 = "Constraint - dyn_l_01_collar_MSTR (P)"(ComponentPose=@n110, Alpha=1, ConstraintWeights_0=1) {
    alphaBoolBlend {blendOption=Linear}
    boneToModify {boneName=dyn_l_01_collar_MSTR}
    constraintSetup [{targetBone {boneName=neck_01}}]
    constraintWeights [1]
   }
   n112 = "Transform (Modify) Bone - Bone: dyn_shirt_d_MSTR"!(ComponentPose=@n111, bAlphaBoolEnabled:bool=True, Translation="0,0,0", Rotation="0,-10,0", Scale="1,1,1") -> (Pose) {
    alphaInputType=Bool
    rotationMode=BMM_Additive
    scaleMode=BMM_Additive
    translationMode=BMM_Additive
    alphaBoolBlend {blendInTime=0.2, blendOption=Linear, blendOutTime=0.1}
    boneToModify {boneName=dyn_shirt_d_MSTR}
    rotation {yaw=-10}
   }
   n113 = AND Boolean(A=Get bIsCrouching(), B=Get bIsAccelerating2D())
   n114 = "Transform (Modify) Bone - Bone: dyn_jacket_g_MSTR"(ComponentPose=@n112, bAlphaBoolEnabled=@n113, Translation="-3,0,0", Rotation="-40,20,10", Scale="1,1,1") {
    alphaInputType=Bool
    rotationMode=BMM_Additive
    scaleMode=BMM_Additive
    translationMode=BMM_Additive
    alphaBoolBlend {blendInTime=0.3, blendOption=Linear, blendOutTime=0.3}
    boneToModify {boneName=dyn_jacket_g_MSTR}
    rotation=<rot -40 20 10>
    translation {x=-3}
   }
   n115 = "Transform (Modify) Bone - Bone: dyn_jacket_f_MSTR"!(ComponentPose=@n114, bAlphaBoolEnabled:bool=@n113, Translation="0,0,3", Rotation="0,0,0", Scale="1,1,1") -> (Pose) {
    alphaInputType=Bool
    rotationMode=BMM_Additive
    scaleMode=BMM_Additive
    translationMode=BMM_Additive
    alphaBoolBlend {blendInTime=0.2, blendOption=Linear, blendOutTime=0.1}
    boneToModify {boneName=dyn_jacket_f_MSTR}
    translation {z=3}
   }
   n116 = "Transform (Modify) Bone - Bone: dyn_l_02_collar_MSTR"!(ComponentPose=@n115, bAlphaBoolEnabled:bool=True, Translation="0.5,0,1.5", Rotation="0,0,0", Scale="1,1,1") -> (Pose) {
    alphaInputType=Bool
    rotationMode=BMM_Additive
    rotationSpace=BCS_BoneSpace
    scaleMode=BMM_Additive
    scaleSpace=BCS_BoneSpace
    translationMode=BMM_Additive
    translationSpace=BCS_BoneSpace
    alphaBoolBlend {blendInTime=0.2, blendOption=Linear, blendOutTime=0.1}
    boneToModify {boneName=dyn_l_02_collar_MSTR}
    translation {x=0.5, z=1.5}
   }
   Output Pose(Result=Component To Local(ComponentPose=@n116)) {layerGroup=None}
  ]}
 EventGraph {nodes [
   Event BlueprintUpdateAnimation()
   > n1 = Sequence!(execute) -> (then_0, then_1, then_2:Exec, then_3:Exec, then_4:Exec, then_5:Exec)
   > n2 = Gravity Override State Function DEPRECATED(CharacterPartAnimInstance=Self-Reference(), GameTimeInSeconds=GetGameTimeInSeconds(), GravityOverrideParams=Get Gravity Override Params())
   > Set GravityOverride(GravityOverride=@n2)
   n3 = Break Fort Anim Input Slope Info!(FortAnimInput_SlopeInfo:"Fort Anim Input Slope Info Structure (by ref)"=Get SlopeInfoInput() -> (SlopeInfoInput:Fort Anim Input Slope Info Structure)) -> (RootSlopeTranslation:Vector=~, LeftFootSlopeTranslation:Vector=~, RightFootSlopeTranslation:Vector=~, SlopeRotation:Rotator=~) {is=title K2Node_BreakStruct}
   Set slope_root_trans(execute=@n1.then_1, slope_root_trans:Vector=@n3) -> (Output_Get:Vector)
   > Set slope_foot_rot(slope_foot_rot:Rotator=@n3.SlopeRotation) -> (Output_Get:Rotator)
   n4 = ComplexSimSpaceFunction DEPRECATED!(
    execute=@n1.then_2
    CharacterPartAnimInstance:Custom Character Part Anim Instance Object Reference=Self-Reference()
    InterpSpeedMax:double=9
    PawnSpeedMax:double=20
    RInterpErrorTolerance:double=1
    DeltaTime:double=0
    DebugDraw:bool=false
    sim_space_rotation_in:Rotator=Get sim_space_rotation()
    inComponentSpaceIn:bool=Get InComponentSpace() -> (InComponentSpace:bool)
    inComponentSpaceLastFrameIn:bool=Get InComponentSpaceLastFrame() -> (InComponentSpaceLastFrame:bool)
    lockedOnSimSpaceIn:bool=Get LockedOnSimSpace() -> (LockedOnSimSpace:bool)
    InterpSpeedIn:double=Get InterpSpeed() -> (InterpSpeed:double)
   ) -> (
    then
    sim_space_rotation:Rotator
    inComponentSpaceOut:bool
    inComponentSpaceLastFrameOut:bool
    lockedOnSimSpaceOut:bool
    InterpSpeedOut:double
   )
   > Set sim_space_rotation(sim_space_rotation:Rotator=@n4) -> (Output_Get:Rotator)
   > Set InComponentSpace(InComponentSpace:bool=@n4.inComponentSpaceOut) -> (Output_Get:bool)
   > Set InComponentSpaceLastFrame(InComponentSpaceLastFrame:bool=@n4.inComponentSpaceLastFrameOut) -> (Output_Get:bool)
   > Set LockedOnSimSpace(LockedOnSimSpace:bool=@n4.lockedOnSimSpaceOut) -> (Output_Get:bool)
   > Set InterpSpeed(InterpSpeed:double=@n4.InterpSpeedOut) -> (Output_Get:double)
   Branch(execute=@n1.then_3, Condition=OR Boolean(A=Get bIsSkyDiving(), B=Get bIsInFrontEnd()))
   > Set frontEndorSkydiving(frontEndorSkydiving:bool=true) -> (Output_Get:bool)
   n5 = GetOwnerSkeletalMeshComponent()
   n6 = BreakTransform(InTransform=GetDeltaTransformFromRefPose(self=@n5, BoneName=upperarm_r, BaseName=clavicle_r))
   Set LookUp(execute=@n1.then_4, LookUp:double=MapRangeClamped(Value=BreakRotator(InRot=@n6.Rotation).Yaw, InRangeA=90, InRangeB=120)) -> (Output_Get:double)
   n7 = BreakTransform(InTransform=GetDeltaTransformFromRefPose(self=@n5, BoneName=upperarm_r, BaseName=clavicle_r))
   > Set Lookup_Collar(Lookup_Collar:double=MapRangeClamped(Value=BreakRotator(InRot=@n7.Rotation).Yaw, InRangeA=40, InRangeB=90)) -> (Output_Get:double)
   n8 = BreakTransform(InTransform=GetDeltaTransformFromRefPose(self=@n5, BoneName=upperarm_l, BaseName=clavicle_l))
   > Set LookDown(LookDown:double=MapRangeClamped(Value=BreakRotator(InRot=@n8.Rotation).Yaw, InRangeA=50, InRangeB=15)) -> (Output_Get:double)
   n9 = BreakTransform(InTransform=GetDeltaTransformFromRefPose(self=@n5, BoneName=upperarm_r, BaseName=clavicle_r))
   > Set LookUpRifle(LookUpRifle:double=MapRangeClamped(Value=BreakRotator(InRot=@n9.Rotation).Yaw, InRangeA=60, InRangeB=78)) -> (Output_Get:double)
   n10 = GetPartSkeletalMeshComponent()
   Is Valid(exec=@n1.then_5, InputObject=@n10)
   > SetVectorParameterValueOnMaterials!(
    execute
    self:Mesh Component Object Reference=@n10
    ParameterName:Name=DO NOT OVERRIDE-Virtual-Pivot-Ch
    ParameterValue:Vector=TransformLocation!(T:"Transform (by ref)"=GetSocketTransform(self=@n10, InSocketName=spine_05, TransformSpace=RTS_World), Location:Vector="-5, 0,0") -> (ReturnValue:Vector)
   ) -> (then)
   > SetScalarParameterValueOnMaterials!(execute, self:Mesh Component Object Reference=@n10, ParameterName:Name=DO NOT OVERRIDE-Character-Position-BP-Switch, ParameterValue:float=1) -> (then)
   Event BlueprintInitializeAnimation()
   > Set Skel_Mesh(Skel_Mesh=GetOwnerSkeletalMeshComponent())
   Event OnEndCrouching()
   > Set crouch alpha(crouch alpha:double=0) -> (Output_Get:double)
   Event OnBeginCrouching()
   > Set crouch alpha(crouch alpha=1)
   PrintString!(execute, InString="To String (Float)"(InDouble=0), bPrintToScreen, bPrintToLog, TextColor, Duration, Key) -> (then)
  ]}
]
variables [
 crouch alpha {type=number}
 frontEndorSkydiving {type=boolean}
 Gravity Override Params {type=GravityOverrideParamsStruct}
 GravityOverride {type=Vector}
 InComponentSpace {type=boolean}
 InComponentSpaceLastFrame {type=boolean}
 InterpSpeed {type=number}
 LockedOnSimSpace {type=boolean}
 LookDown {type=number}
 LookUp {type=number}
 Lookup_Collar {type=number}
 LookUpRifle {type=number}
 sim_space_rotation {type=Rotator}
 Skel_Mesh {type=SkeletalMeshComponent}
 slope_foot_rot {type=Rotator}
 slope_root_trans {type=Vector}
]
